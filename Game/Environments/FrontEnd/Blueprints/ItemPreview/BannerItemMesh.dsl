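; Front-end item preview mesh actor for banners, showing the previewed banner icon on its mesh material.
; On BeginPlay it casts its parent actor to FortItemPreviewOffPawnActor, resolves that actor's preview icon and feeds it to a dynamic material's "WorkerPortrait" texture parameter.
; It also sets flip and position parameters to align the icon.
; The zoomed-out camera framing override scales the parent's framing cylinder height and radius by BoundScale.
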(event EventBeginPlay
  (bind _asfort_item_preview_off_pawn_actor (Utilities|Casting|CastToFortItemPreviewOffPawnActor (Actor|GetParentActor))) ; Expected to be spawned by BannerItemPreview
  (bind _returnvalue (Rendering|Material|CreateDynamicMaterialInstance (Variables|Default|GetItemMesh) 0 0 "None" :!class /Script/Engine.PrimitiveComponent))
  (Rendering|Material|SetTextureParameterValue _returnvalue "WorkerPortrait" (Utilities|ResolveSoftReference (Class|FortItemPreviewOffPawnActor|GetPreviewIcon _asfort_item_preview_off_pawn_actor)))
  (Rendering|Material|SetScalarParameterValue _returnvalue "bFlipPortrait" 1.0 :!class /Script/Engine.MaterialInstanceDynamic)
  ;[1 1 1] Aligning banner icons
  (Rendering|Material|SetScalarParameterValue _returnvalue "Position X" -0.025 :!class /Script/Engine.MaterialInstanceDynamic)
  (Rendering|Material|SetScalarParameterValue _returnvalue "Position Y" -0.05 :!class /Script/Engine.MaterialInstanceDynamic))

(fn GetZoomedOutCameraFramingSettings
  (bind _returnvalue (|CallParentFunction))
  (bind (_returnvalue_focuspointtocenter _returnvalue_boundingbehavior _returnvalue_boundstoframe _returnvalue_focusobjectrotation) (Utilities|Struct|BreakFortUICameraFrameTargetSettings _returnvalue))
  (bind (_returnvalue_boundstoframe_origin _returnvalue_boundstoframe_cylinderhalfheight _returnvalue_boundstoframe_cylinderradius) (Utilities|Struct|BreakFortUICameraFrameTargetBounds _returnvalue_boundstoframe))
  (return _returnvalue_focuspointtocenter _returnvalue_boundingbehavior _returnvalue_boundstoframe_origin (* _returnvalue_boundstoframe_cylinderhalfheight (Variables|Default|GetBoundScale)) (* _returnvalue_boundstoframe_cylinderradius (Variables|Default|GetBoundScale)) _returnvalue_focusobjectrotation))

(fn ConstructionScript)
