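; Battle pass vault preview world Blueprint (front end) that drives the animated backdrop and floor behind previewed items.
; It creates dynamic materials for background and floor, writing each new texture, colors and offset into the inactive of two material pages (A/B).
; Timelines crossfade and slide between pages by direction, then flip the active page.
; Other timelines animate streak effects, floor mask, item-detail offsets and per-page texture intensity.
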(event EventBeginPlay
  ;[0 0 0] BeginPlay
  (|CallParentFunction)
  (bind _returnvalue (Rendering|Material|CreateDynamicMaterialInstance (Variables|Default|GetFloorMI) "None" "None" :!class /Script/Engine.KismetMaterialLibrary))
  (Rendering|Material|SetMaterial (Variables|Default|GetFloor) 0 _returnvalue)
  (bind _output_get (Variables|Default|SetFloorMID _returnvalue))
  (CallFunction|SetFloorMaterialDynamic _output_get)
  ;[1 1 1] Set length of dissolve time and start the loop. This would probably be better on the event of when your camera switches to the bpass vault world view instead of begin play here.
  (Utilities|FlowControl|Delay 5.0)
  (Components|Timeline|SetPlayRate (Variables|BattlePassVaultWorld|GetIconDissolve) (/ 1.0 15.0))
  (Components|Timeline|PlayfromStart (Variables|BattlePassVaultWorld|GetIconDissolve)))

(event EventOnSetupTextureBackground
  (params (LoadedBackgroundTexture :type Texture%202D%20Object%20Reference) (BackgroundInfo :type Vault%20World%20Background%20Data%20Structure%20%28by%20ref%29))
  ;[0 0 0] OnSetupBackground
  (Components|Timeline|Stop (Variables|BattlePassVaultWorld|GetTransitionForward))
  (CallFunction|SetupBackgroundMaterial LoadedBackgroundTexture BackgroundInfo (Variables|FortItemPreviewWorld|GetBackgroundMaterialDynamic))
  (CallFunction|SetupBackgroundMaterial LoadedBackgroundTexture BackgroundInfo (Variables|Default|GetFloorMID)))

(event EventOnSetupMaterialBackground
  (params (LoadedBackgroundMaterial :type Material%20Interface%20Object%20Reference) (BackgroundInfo :type Vault%20World%20Background%20Data%20Structure%20%28by%20ref%29))
  ;[0 0 0] OnSetupBackground
  (Components|Timeline|Stop (Variables|BattlePassVaultWorld|GetTransitionForward))
  (CallFunction|SetupBackgroundMaterial "" BackgroundInfo (Variables|FortItemPreviewWorld|GetBackgroundMaterialDynamic))
  (CallFunction|SetupBackgroundMaterial "" BackgroundInfo (Variables|Default|GetFloorMID)))

(event EventOnTransitionBackground
  (params (TransitionDirection :type EVaultWorldTransitionDirection%20Enum :default "Forward") (IntensityTransition :type EBackgroundIntensityLevel%20Enum))
  ;[0 0 0] OnTransitionBackground
  (Variables|Default|SetIntensity%20Transition IntensityTransition)
  (Variables|Default|SetbIsTransitioning true)
  (switch Utilities|FlowControl|Switch|SwitchonEVaultWorldTransitionDirection TransitionDirection
    (:Forward
      (CallFunction|PageTransitionIntensityUpdate (Variables|FortItemPreviewWorld|GetBackgroundMaterialDynamic) (Variables|Default|GetFloorMID))
      (bind (_direction _forward _fade _fx_transition_fade) (Timeline|TransitionForward PlayFromStart)
        (:Update
          (Utilities|FlowControl|Sequence
            (:then_0
              (CallFunction|TransitionBackgroundForward _forward _fade _fx_transition_fade (Variables|FortItemPreviewWorld|GetBackgroundMaterialDynamic)))
            (:then_1
              (CallFunction|TransitionBackgroundForward _forward _fade _fx_transition_fade (Variables|Default|GetFloorMID)))))
        (:Finished
          (Variables|Default|SetbIsPrimaryBackgroundActive (not (Variables|Default|GetbIsPrimaryBackgroundActive)))
          (Variables|Default|SetbIsTransitioning false))))
    (:Backward
      (CallFunction|PageTransitionIntensityUpdate (Variables|FortItemPreviewWorld|GetBackgroundMaterialDynamic) (Variables|Default|GetFloorMID))
      (bind (_direction_2 _backward _fade_2 _fx_transition_fade_2) (Timeline|TransitionBackward PlayFromStart)
        (:Update
          (Utilities|FlowControl|Sequence
            (:then_0
              (CallFunction|TransitionBackgroundBackward _backward _fade_2 _fx_transition_fade_2 (Variables|FortItemPreviewWorld|GetBackgroundMaterialDynamic)))
            (:then_1
              (CallFunction|TransitionBackgroundBackward _backward _fade_2 _fx_transition_fade_2 (Variables|Default|GetFloorMID)))))
        (:Finished
          (Variables|Default|SetbIsPrimaryBackgroundActive (not (Variables|Default|GetbIsPrimaryBackgroundActive)))
          (Variables|Default|SetbIsTransitioning false))))
    (:None
      ;[0 0 0] OnFirstTimeBackgroundSetup
      (Utilities|FlowControl|Sequence
        (:then_0
          (CallFunction|TransitionBackgroundForward 1.0 1.0 1.0 (Variables|FortItemPreviewWorld|GetBackgroundMaterialDynamic)))
        (:then_1
          (CallFunction|TransitionBackgroundForward 1.0 1.0 1.0 (Variables|Default|GetFloorMID)))
        (:then_2
          (Variables|Default|SetbIsPrimaryBackgroundActive (not (Variables|Default|GetbIsPrimaryBackgroundActive))))))))

(event EventOnUpdateDisplay
  (params (bShowFloor :type Boolean) (bShowEffects :type Boolean) (bIsOnPawn :type Boolean) (ItemType :type EFortItemType%20Enum :default "WorldItem"))
  ;[0 0 0] OnShowBackgroundEffects
  (Utilities|FlowControl|Sequence
    (:then_0
      (if bShowEffects
        (bind (_direction _setstreaks) (Timeline|Background-Effects Reverse)
          (:Update
            (Rendering|Material|SetScalarParameterValue (Variables|FortItemPreviewWorld|GetBackgroundMaterialDynamic) "bIsUseStreak" _setstreaks :!class /Script/Engine.MaterialInstanceDynamic)
            (Rendering|Material|SetScalarParameterValue (Variables|Default|GetFloorMID) "bIsUseStreak" _setstreaks :!class /Script/Engine.MaterialInstanceDynamic)))
        (else
          (Timeline|Background-Effects PlayFromStart))))
    (:then_1
      (if bShowFloor
        (bind (_direction_2 _floormask) (Timeline|Floor-Visibility PlayFromStart)
          (:Update
            (Rendering|Material|SetScalarParameterValue (Variables|Default|GetFloorMID) "Global-FloorMask-Density" _floormask :!class /Script/Engine.MaterialInstanceDynamic))
          (:Finished
            _)))
      (Rendering|SetVisibility (Variables|Default|GetFloor) bShowFloor true))))

(event EventOnTransitionItemDetails
  (params (bShowItemDetails :type Boolean))
  (bind _backgroundmaterialdynamic (Variables|FortItemPreviewWorld|GetBackgroundMaterialDynamic))
  (bind _bg_intensity_mid (Variables|Default|GetBG_Intensity_Mid))
  (bind _bg_intensity_min (Variables|Default|GetBG_Intensity_Min))
  ;[0 0 0] OnTransitionItemDetails
  (Utilities|FlowControl|Sequence
    (:then_0
      (if bShowItemDetails
        (bind (_direction _x_offset _icon_x_offset) (Timeline|ItemDetails PlayFromStart)
          (:Update
            (Rendering|Material|SetScalarParameterValue (Variables|FortItemPreviewWorld|GetBackgroundMaterialDynamic) "Cnt_Glow_x-Offset" _x_offset :!class /Script/Engine.MaterialInstanceDynamic)
            (Rendering|Material|SetScalarParameterValue (Variables|Default|GetFloorMID) "Icon-X-Offset" _icon_x_offset :!class /Script/Engine.MaterialInstanceDynamic)))
        (else
          (Timeline|ItemDetails Reverse))))
    (:then_1
      (if bShowItemDetails
        (bind (_direction_2 _textureintensity) (Timeline|ItemDetailsIntensity PlayFromStart)
          (:Update
            (if (Variables|Default|GetbIsPrimaryBackgroundActive)
              (Rendering|Material|SetScalarParameterValue _backgroundmaterialdynamic "PageA Texture Intensity" (Math|Float|Lerp _bg_intensity_min _bg_intensity_mid _textureintensity) :!class /Script/Engine.MaterialInstanceDynamic)
              (Rendering|Material|SetScalarParameterValue (Variables|Default|GetFloorMID) "Bokeh Intensity" (Math|Float|Lerp 1.0 0.0 _textureintensity) :!class /Script/Engine.MaterialInstanceDynamic)
              (else
                (Rendering|Material|SetScalarParameterValue _backgroundmaterialdynamic "PageB Texture Intensity" (Math|Float|Lerp _bg_intensity_min _bg_intensity_mid _textureintensity) :!class /Script/Engine.MaterialInstanceDynamic)
                (Rendering|Material|SetScalarParameterValue _backgroundmaterialdynamic "Bokeh Intensity" (Math|Float|Lerp 1.0 0.0 _textureintensity) :!class /Script/Engine.MaterialInstanceDynamic)))))
        (else
          (Timeline|ItemDetailsIntensity Reverse))))))

(event EventOnIntensityChange
  (params (bToLowIntensity :type Boolean))
  (bind _backgroundmaterialdynamic (Variables|FortItemPreviewWorld|GetBackgroundMaterialDynamic))
  (bind _bg_intensity_max (Variables|Default|GetBG_Intensity_Max))
  (bind _bg_intensity_mid (Variables|Default|GetBG_Intensity_Mid))
  ;[0 0 0] Set Intensity
  (if bToLowIntensity
    (bind (_direction _intensity) (Timeline|BackgroundIntenstiy PlayFromStart)
      (:Update
        (if (xor (Variables|Default|GetbIsPrimaryBackgroundActive) (Variables|Default|GetbIsTransitioning))
          (Rendering|Material|SetScalarParameterValue _backgroundmaterialdynamic "PageA Texture Intensity" (Math|Float|Lerp _bg_intensity_mid _bg_intensity_max _intensity) :!class /Script/Engine.MaterialInstanceDynamic)
          (else
            (Rendering|Material|SetScalarParameterValue _backgroundmaterialdynamic "PageB Texture Intensity" (Math|Float|Lerp _bg_intensity_mid _bg_intensity_max _intensity) :!class /Script/Engine.MaterialInstanceDynamic))))
      (:Finished
        (switch Utilities|FlowControl|Switch|SwitchonETimelineDirection _direction
          (:Forward
            (if (Variables|Default|GetbIsPrimaryBackgroundActive)
              (Variables|Default|SetbIsPageA_Max false)
              (else
                (Variables|Default|SetbIsPageB_Max false))))
          (:Backward
            (if (Variables|Default|GetbIsPrimaryBackgroundActive)
              (Variables|Default|SetbIsPageA_Max true)
              (else
                (Variables|Default|SetbIsPageB_Max true)))))))
    (else
      (Timeline|BackgroundIntenstiy ReverseFromEnd))))

(event EventOnNewSceneBackgroundChildActor
  (params (NewActor :type Fort%20Static%20Mesh%20Actor%20Object%20Reference))
  (CallFunction|AssignMaterials NewActor))

(fn AssignMaterials
  (params (TargetBackground :type Static%20Mesh%20Actor%20Object%20Reference))
  (bind _returnvalue (Rendering|Material|CreateDynamicMaterialInstance (Variables|FortItemPreviewWorld|GetBackgroundMaterial) "None" "None" :!class /Script/Engine.KismetMaterialLibrary))
  (Rendering|Material|SetMaterial (Variables|/Script/Engine.StaticMeshActor|GetStaticMeshComponent TargetBackground) 0 _returnvalue)
  (CallFunction|SetBackgroundMaterialDynamic _returnvalue)
  (CallFunction|SetFloorMaterialDynamic (Variables|Default|GetFloorMID)))

(fn PageTransitionIntensityUpdate
  (params (MID_Background :type Material%20Instance%20Dynamic%20Object%20Reference) (MID_Floor :type Material%20Instance%20Dynamic%20Object%20Reference))
  (bind _intensity_transition (Variables|Default|GetIntensity%20Transition))
  (bind _bg_intensity_max (Variables|Default|GetBG_Intensity_Max))
  (bind _bg_intensity_mid (Variables|Default|GetBG_Intensity_Mid))
  (bind _returnvalue (Utilities|Select :Index _intensity_transition :LowIntensity _bg_intensity_mid :HighIntensity _bg_intensity_max))
  (bind _intensity_transition_2 (Variables|Default|GetIntensity%20Transition))
  (bind _returnvalue_2 (Utilities|Select :Index _intensity_transition_2 :LowIntensity false :HighIntensity true))
  (if (Variables|Default|GetbIsPrimaryBackgroundActive)
    (if (!= (Variables|Default|GetIntensity%20Transition) 0)
      (Rendering|Material|SetScalarParameterValue (Variables|GetMID_Background) "PageB Texture Intensity" _returnvalue :!class /Script/Engine.MaterialInstanceDynamic)
      (Variables|Default|SetbIsPageB_Max _returnvalue_2)
      (else
        (Rendering|Material|SetScalarParameterValue (Variables|GetMID_Background) "PageB Texture Intensity" (select (Variables|Default|GetbIsPageA_Max) (Variables|Default|GetBG_Intensity_Max) (Variables|Default|GetBG_Intensity_Mid)) :!class /Script/Engine.MaterialInstanceDynamic)
        (Variables|Default|SetbIsPageB_Max (Variables|Default|GetbIsPageA_Max))))
    (elif (!= (Variables|Default|GetIntensity%20Transition) 0)
      (Rendering|Material|SetScalarParameterValue (Variables|GetMID_Background) "PageA Texture Intensity" _returnvalue :!class /Script/Engine.MaterialInstanceDynamic)
      (Variables|Default|SetbIsPageA_Max _returnvalue_2)
      (else
        (Rendering|Material|SetScalarParameterValue (Variables|GetMID_Background) "PageA Texture Intensity" (select (Variables|Default|GetbIsPageB_Max) (Variables|Default|GetBG_Intensity_Max) (Variables|Default|GetBG_Intensity_Mid)) :!class /Script/Engine.MaterialInstanceDynamic)
        (Variables|Default|SetbIsPageA_Max (Variables|Default|GetbIsPageB_Max))))))

(fn SetBackgroundsScalarValue
  (params (Param :type Name) (FloatValue :type Float%20%28double-precision%29))
  (Rendering|Material|SetScalarParameterValue (Variables|FortItemPreviewWorld|GetBackgroundMaterialDynamic) Param FloatValue :!class /Script/Engine.MaterialInstanceDynamic))

(fn SetupBackgroundMaterial
  (params (TextureBackground :type Texture%202D%20Object%20Reference%20%28by%20ref%29) (BackgroundInfo :type Vault%20World%20Background%20Data%20Structure%20%28by%20ref%29) (MID :type Material%20Instance%20Dynamic%20Object%20Reference))
  (bind (_backgroundsubstance _backgroundoffset _primarycolor _secondarycolor) (Utilities|Struct|BreakVaultWorldBackgroundData BackgroundInfo))
  (bind _x (.x _backgroundoffset))
  (bind _bisprimarybackgroundactive (Variables|Default|GetbIsPrimaryBackgroundActive))
  (Rendering|Material|SetTextureParameterValue MID (select _bisprimarybackgroundactive "PageB-Texture" "PageA-Texture") (select (Utilities|IsValid TextureBackground :!class /Script/Engine.KismetSystemLibrary) TextureBackground "/Game/Athena/UI/Frontend/Spatial/Art/T_Ui_Empty.T_Ui_Empty"))
  (Rendering|Material|SetVectorParameterValue MID (select _bisprimarybackgroundactive "PageB-RBG1" "PageA-RBG1") _primarycolor :!class /Script/Engine.MaterialInstanceDynamic)
  (Rendering|Material|SetVectorParameterValue MID (select _bisprimarybackgroundactive "PageB-RBG2" "PageA-RBG2") _secondarycolor :!class /Script/Engine.MaterialInstanceDynamic)
  (Rendering|Material|SetScalarParameterValue MID (select _bisprimarybackgroundactive "PageB Texture Offset X" "PageA Texture Offset X") _x :!class /Script/Engine.MaterialInstanceDynamic)
  (Rendering|Material|SetScalarParameterValue MID (select _bisprimarybackgroundactive "PageB Texture Offset Y" "PageA Texture Offset Y") (.y _backgroundoffset) :!class /Script/Engine.MaterialInstanceDynamic)
  (if (Variables|Default|GetbIsPrimaryBackgroundActive)
    (Variables|Default|SetPageB_X_Offset _x)
    (else
      (Variables|Default|SetPageA_X_Offset _x))))

(fn TransitionBackgroundBackward
  (params (Backward :type Float%20%28double-precision%29) (Fade :type Float%20%28double-precision%29) (FXTransitionFade :type Float%20%28double-precision%29) (MID :type Material%20Instance%20Dynamic%20Object%20Reference))
  (bind _fxtransitionfade (Variables|GetFXTransitionFade))
  (bind _fade (Variables|GetFade))
  (Rendering|Material|SetScalarParameterValue (Variables|GetMID) "IsPrimaryBackgroundActive" (Math|Conversions|ToFloat(Boolean) (Variables|Default|GetbIsPrimaryBackgroundActive)) :!class /Script/Engine.MaterialInstanceDynamic)
  (if (Variables|Default|GetbIsPrimaryBackgroundActive)
    ;[0 0 0 1 12] Forward A to B
    (Rendering|Material|SetScalarParameterValue (Variables|GetMID) "Transtion-Amount" (Math|Float|Lerp 0.0 1.0 _fade) :!class /Script/Engine.MaterialInstanceDynamic)
    ;[0 0 0 1 12] Offset Textures
    (Rendering|Material|SetScalarParameterValue (Variables|GetMID) "PageA Texture Offset X" (Math|Float|Lerp (Variables|Default|GetPageA_X_Offset) (- (Variables|Default|GetPageA_X_Offset) (Variables|Default|GetX_Offset_Anim_Distance)) (Variables|GetBackward)) :!class /Script/Engine.MaterialInstanceDynamic)
    (Rendering|Material|SetScalarParameterValue (Variables|GetMID) "PageB Texture Offset X" (Math|Float|Lerp (+ (Variables|Default|GetX_Offset_Anim_Distance) (Variables|Default|GetPageB_X_Offset)) (Variables|Default|GetPageB_X_Offset) (Variables|GetBackward)) :!class /Script/Engine.MaterialInstanceDynamic)
    ;[0 0 0 1 12] Fade FX
    (Rendering|Material|SetScalarParameterValue (Variables|GetMID) "fx-Transition-Fade" _fxtransitionfade :!class /Script/Engine.MaterialInstanceDynamic)
    (else
      ;[0 0 0 1 12] Forward B to A
      (Rendering|Material|SetScalarParameterValue (Variables|GetMID) "Transtion-Amount" (Math|Float|Lerp 1.0 0.0 _fade) :!class /Script/Engine.MaterialInstanceDynamic)
      (Rendering|Material|SetScalarParameterValue (Variables|GetMID) "PageA Texture Offset X" (Math|Float|Lerp (+ (Variables|Default|GetX_Offset_Anim_Distance) (Variables|Default|GetPageA_X_Offset)) (Variables|Default|GetPageA_X_Offset) (Variables|GetBackward)) :!class /Script/Engine.MaterialInstanceDynamic)
      (Rendering|Material|SetScalarParameterValue (Variables|GetMID) "PageB Texture Offset X" (Math|Float|Lerp (Variables|Default|GetPageB_X_Offset) (- (Variables|Default|GetPageB_X_Offset) (Variables|Default|GetX_Offset_Anim_Distance)) (Variables|GetBackward)) :!class /Script/Engine.MaterialInstanceDynamic)
      (Rendering|Material|SetScalarParameterValue (Variables|GetMID) "fx-Transition-Fade" _fxtransitionfade :!class /Script/Engine.MaterialInstanceDynamic))))

(fn TransitionBackgroundForward
  (params (Forward :type Float%20%28double-precision%29) (Fade :type Float%20%28double-precision%29) (FXTransitionFade :type Float%20%28double-precision%29) (MID :type Material%20Instance%20Dynamic%20Object%20Reference))
  (bind _fxtransitionfade (Variables|GetFXTransitionFade))
  (bind _fade (Variables|GetFade))
  (Rendering|Material|SetScalarParameterValue (Variables|GetMID) "IsPrimaryBackgroundActive" (Math|Conversions|ToFloat(Boolean) (Variables|Default|GetbIsPrimaryBackgroundActive)) :!class /Script/Engine.MaterialInstanceDynamic)
  (if (Variables|Default|GetbIsPrimaryBackgroundActive)
    ;[0 0 0 1 12] Forward A to B
    (Rendering|Material|SetScalarParameterValue (Variables|GetMID) "Transtion-Amount" (Math|Float|Lerp 0.0 1.0 _fade) :!class /Script/Engine.MaterialInstanceDynamic)
    ;[0 0 0 1 12] Offset Textures
    (Rendering|Material|SetScalarParameterValue (Variables|GetMID) "PageA Texture Offset X" (Math|Float|Lerp (Variables|Default|GetPageA_X_Offset) (+ (Variables|Default|GetPageA_X_Offset) (Variables|Default|GetX_Offset_Anim_Distance)) (Variables|GetForward)) :!class /Script/Engine.MaterialInstanceDynamic)
    (Rendering|Material|SetScalarParameterValue (Variables|GetMID) "PageB Texture Offset X" (Math|Float|Lerp (- (Variables|Default|GetPageB_X_Offset) (Variables|Default|GetX_Offset_Anim_Distance)) (Variables|Default|GetPageB_X_Offset) (Variables|GetForward)) :!class /Script/Engine.MaterialInstanceDynamic)
    ;[0 0 0 1 12] Fade FX
    (Rendering|Material|SetScalarParameterValue (Variables|GetMID) "fx-Transition-Fade" _fxtransitionfade :!class /Script/Engine.MaterialInstanceDynamic)
    (else
      ;[0 0 0 1 12] Forward B to A
      (Rendering|Material|SetScalarParameterValue (Variables|GetMID) "Transtion-Amount" (Math|Float|Lerp 1.0 0.0 _fade) :!class /Script/Engine.MaterialInstanceDynamic)
      (Rendering|Material|SetScalarParameterValue (Variables|GetMID) "PageA Texture Offset X" (Math|Float|Lerp (- (Variables|Default|GetPageA_X_Offset) (Variables|Default|GetX_Offset_Anim_Distance)) (Variables|Default|GetPageA_X_Offset) (Variables|GetForward)) :!class /Script/Engine.MaterialInstanceDynamic)
      (Rendering|Material|SetScalarParameterValue (Variables|GetMID) "PageB Texture Offset X" (Math|Float|Lerp (Variables|Default|GetPageB_X_Offset) (+ (Variables|Default|GetPageB_X_Offset) (Variables|Default|GetX_Offset_Anim_Distance)) (Variables|GetForward)) :!class /Script/Engine.MaterialInstanceDynamic)
      (Rendering|Material|SetScalarParameterValue (Variables|GetMID) "fx-Transition-Fade" _fxtransitionfade :!class /Script/Engine.MaterialInstanceDynamic))))

(fn ConstructionScript
  (Rendering|Material|SetMaterial (Variables|Default|GetFloor) 0 (Variables|Default|GetFloorMI))
  (bind _asfort_static_mesh_actor (Utilities|Casting|CastToFortStaticMeshActor (Variables|/Script/Engine.ChildActorComponent|GetChildActor (Variables|FortItemPreviewWorld|GetSceneBackgroundComponent))))
  (Rendering|Material|SetMaterial (Variables|/Script/Engine.StaticMeshActor|GetStaticMeshComponent _asfort_static_mesh_actor) 0 (Variables|FortItemPreviewWorld|GetBackgroundMaterial))
  (|CallParentFunction))
