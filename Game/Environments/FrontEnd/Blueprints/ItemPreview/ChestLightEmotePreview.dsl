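; Front-end item preview actor that previews items on a pawn with its own lighting rig and floor plane, apparently for emotes.
; When visuals spawn it plays a transition sound and particles, attaching a looping obscuring emitter to the pawn and destroying it on later preview events.
; Unless AlwaysOn, light functions toggle PC, mobile and Switch lights by platform and shadow quality, re-evaluated on settings changes while active.
; The floor plane receives a dynamic material with parameters copied from the supplied instance.
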(event EventBeginPlay)

(event Collision|EventActorBeginOverlap
  (params (OtherActor :type Actor%20Object%20Reference)))

(event EventTick
  (params (DeltaSeconds :type Float%20%28single-precision%29)))

(event EventOnPreviewVisualsSpawned
  (bind _returnvalue (CallFunction|GetPawnLocation))
  (Components|DestroyComponent (Variables|Default|GetObscuringLoopEmitter))
  (Effects|SpawnEmitteratLocation "/Game/Effects/Fort_Effects/UI/Transition/P_FrontEndCharacterTransition.P_FrontEndCharacterTransition" _returnvalue)
  (bind _returnvalue_2 (Effects|SpawnEmitteratLocation "/Game/Effects/Fort_Effects/UI/Transition/P_FrontEndCharacterTransition_Loop.P_FrontEndCharacterTransition_Loop" _returnvalue))
  (bind _output_get (Variables|Default|SetObscuringLoopEmitter _returnvalue_2))
  (Transformation|AttachComponentToComponent _output_get (Variables|FortItemPreviewOnPawnActor|GetPrimaryPreviewPawnComponent) "None" "KeepRelative" "KeepRelative" "KeepRelative" true)
  (Utilities|FlowControl|Sequence
    (:then_0
      (Audio|PlaySoundatLocation "/Game/Sounds/Fort_UI_Sounds/S13_Frontend/BattlePass_ResIn_Cue.BattlePass_ResIn_Cue" _returnvalue)
      (if (not (Variables|Default|GetAlwaysOn))
        (Utilities|FlowControl|Sequence
          (:then_0
            (Rendering|Components|Light|SetForceCachedShadowsforMovablePrimitives (Variables|Default|GetKeyLight5) true))
          (:then_1
            ;[0 0 0 1 12] Forece Disable on start
            (CallFunction|SwitchMobileLighting false)
            (CallFunction|SwitchPCLighting false)
            (CallFunction|LightControl true))
          (:then_2
            ;[0 0 0 1 12] Sitch lighting if settings are switched while screen is already active
            (bind _returnvalue_3 (Settings|GetGameUserSettings))
            (EventDispatchers|BindEventtoOnGameUserSettingsUINeedsUpdate _returnvalue_3 (delegate Custom|UpdateSettings))))))
    (:then_1
      (Utilities|FlowControl|WaitforBothExecutions true
        (:Then
          (Components|DestroyComponent (Variables|Default|GetObscuringLoopEmitter)))))))

(event Custom|UpdateSettings
  (bind _returnvalue (TouchPlatform|IsNintendoSwitch))
  (bind _returnvalue_2 (TouchPlatform|IsMobileGame))
  ;[0 0 0 1 12] Sitch lighting if settings are switched while screen is already active
  (if (Variables|Default|GetIsActive)
    (if (Math|Boolean|ORBoolean (== (Settings|GetShadowQuality (stale EventOnPreviewVisualsSpawned _returnvalue_3)) 0) _returnvalue_2 _returnvalue)
      (if (or _returnvalue_2 _returnvalue)
        ;;[0.125 0.125 0.125 1 12] Enable Mobile
        (CallFunction|SwitchPCLighting false)
        (CallFunction|SwitchMobileLighting true)
        (else
          ;;[0.125 0.125 0.125 1 12] Enable Low Detail PC
          (CallFunction|SwitchPCLighting true)))
      (else
        ;;[0.125 0.125 0.125 1 12] Enable HighEnd
        (CallFunction|SwitchMobileLighting false)
        (CallFunction|SwitchPCLighting true)))))

(event EventOnPreviewVisualsShown
  (Utilities|FlowControl|WaitforBothExecutions true
    (:Then
      (Components|DestroyComponent (Variables|Default|GetObscuringLoopEmitter)))))

(event EventOnCurrentVisualsCleanedUp
  (Components|DestroyComponent (Variables|Default|GetObscuringLoopEmitter))
  (Utilities|FlowControl|WaitforBothExecutions true
    (:Then
      (Components|DestroyComponent (Variables|Default|GetObscuringLoopEmitter)))))

(event EventOnSetFloorMaterial
  (params (InMaterialInstance :type Material%20Interface%20Object%20Reference))
  ;[0 0 0] OnSetFloorMaterial
  (bind _returnvalue (Rendering|Material|CreateDynamicMaterialInstance (Variables|FortItemPreviewActor|GetSpecialEventMaterial) "None" "None" :!class /Script/Engine.KismetMaterialLibrary))
  (Rendering|Material|SetMaterial (Variables|Default|GetPlane) 0 _returnvalue)
  (Rendering|Material|CopyMaterialInstanceParameters _returnvalue InMaterialInstance false))

(fn LightControl
  (params (Active :type Boolean))
  (bind _returnvalue (TouchPlatform|IsNintendoSwitch))
  (bind _returnvalue_2 (TouchPlatform|IsMobileGame))
  (bind _isactive (Variables|Default|GetIsActive))
  (if (not (Variables|Default|GetAlwaysOn))
    (Variables|Default|SetIsActive Active)
    (bind _returnvalue_3 (Settings|GetGameUserSettings))
    (if (Math|Boolean|ORBoolean (== (Settings|GetShadowQuality _returnvalue_3) 0) _returnvalue_2 _returnvalue)
      (if _returnvalue_2
        (CallFunction|SwitchMobileLighting _isactive)
        (elif _returnvalue
          (CallFunction|SwitchErebusLighting _isactive))))
    (CallFunction|SwitchPCLighting (Variables|Default|GetIsActive))))

(fn SetFloorEnabled
  (params (Show%20Floor :type Boolean))
  (Development|SetHiddeninGame (Variables|Default|GetPlane) (not Show%20Floor) true))

(fn SwitchErebusLighting
  (params (Visibility :type Boolean))
  (Rendering|SetVisibility (Variables|Default|GetKeyLight_StandaloneForSwitch) Visibility false))

(fn SwitchMobileLighting
  (params (NewParam :type Boolean))
  (Rendering|SetVisibility (Variables|Default|GetBounce1) false false)
  (Rendering|SetVisibility (Variables|Default|GetDirectionalLightMobile) NewParam false))

(fn SwitchPCLighting
  (params (Visibility :type Boolean))
  (Rendering|SetVisibility (Variables|Default|GetBounce1) Visibility false)
  (Rendering|SetVisibility (Variables|Default|GetKeyLight_StandaloneForSwitch) false false))

(fn ConstructionScript)
