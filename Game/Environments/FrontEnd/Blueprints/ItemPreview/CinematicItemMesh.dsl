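; Front-end item preview Blueprint for a cinematic item mesh that displays the previewed item's icon, expected to be spawned by BannerItemPreview.
; On BeginPlay it casts its parent actor to FortItemPreviewOffPawnActor, resolves the soft preview icon and sets it as the Texture parameter on a dynamic material of ItemMesh.
; Its zoomed-out camera framing override takes the parent function's settings and multiplies the bounds cylinder half-height and radius by BoundScale.
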(event EventBeginPlay
  (bind _asfort_item_preview_off_pawn_actor (Utilities|Casting|CastToFortItemPreviewOffPawnActor (Actor|GetParentActor))) ; Expected to be spawned by BannerItemPreview
  (bind _returnvalue (Rendering|Material|CreateDynamicMaterialInstance (Variables|Default|GetItemMesh) 0 0 "None" :!class /Script/Engine.PrimitiveComponent))
  (Rendering|Material|SetTextureParameterValue _returnvalue "Texture" (Utilities|ResolveSoftReference (Class|FortItemPreviewOffPawnActor|GetPreviewIcon _asfort_item_preview_off_pawn_actor))))

(fn GetZoomedOutCameraFramingSettings
  (bind _returnvalue (|CallParentFunction))
  (bind (_returnvalue_focuspointtocenter _returnvalue_boundingbehavior _returnvalue_boundstoframe _returnvalue_focusobjectrotation) (Utilities|Struct|BreakFortUICameraFrameTargetSettings _returnvalue))
  (bind (_returnvalue_boundstoframe_origin _returnvalue_boundstoframe_cylinderhalfheight _returnvalue_boundstoframe_cylinderradius) (Utilities|Struct|BreakFortUICameraFrameTargetBounds _returnvalue_boundstoframe))
  (return _returnvalue_focuspointtocenter _returnvalue_boundingbehavior _returnvalue_boundstoframe_origin (* _returnvalue_boundstoframe_cylinderhalfheight (Variables|Default|GetBoundScale)) (* _returnvalue_boundstoframe_cylinderradius (Variables|Default|GetBoundScale)) _returnvalue_focusobjectrotation))

(fn ConstructionScript)
