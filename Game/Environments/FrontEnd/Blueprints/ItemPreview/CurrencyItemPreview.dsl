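; Front-end preview actor for a currency item that animates and lights the item when its visuals spawn.
; On spawn it plays transition particles and a sound, while one timeline scales the mesh up and starts another that bobs the root and tilts the mesh.
; Unless AlwaysOn is set, the scale timeline's update also runs LightControl and binds UpdateSettings, toggling lights by platform and shadow quality when user settings change.
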(event EventOnPreviewVisualsSpawned
  (bind _staticmesh (Variables|Default|GetStaticMesh))
  (bind _returnvalue (Transformation|GetWorldLocation _staticmesh))
  (Components|DestroyComponent (Variables|Default|GetObscuringLoopEmitter))
  (Effects|SpawnEmitteratLocation :EmitterTemplate "/Game/Effects/Fort_Effects/UI/Transition/P_FrontEndCharacterTransition.P_FrontEndCharacterTransition" :Location _returnvalue :Scale "0.500000,0.500000,0.500000" :bAutoDestroy true :PoolingMethod "None" :bAutoActivateSystem true)
  (bind _returnvalue_2 (Effects|SpawnEmitteratLocation "/Game/Effects/Fort_Effects/UI/Transition/P_FrontEndCharacterTransition_Loop.P_FrontEndCharacterTransition_Loop" _returnvalue))
  (bind _output_get (Variables|Default|SetObscuringLoopEmitter _returnvalue_2))
  (Transformation|AttachComponentToComponent _output_get (Variables|Default|GetStaticMesh) "None" "KeepRelative" "KeepRelative" "KeepRelative" true)
  (Utilities|FlowControl|Sequence
    (:then_0
      (Audio|PlaySoundatLocation "/Game/Sounds/Fort_UI_Sounds/S13_Frontend/BattlePass_ResIn_Cue.BattlePass_ResIn_Cue" _returnvalue)
      (bind (_direction _lerp) (Timeline|Timeline_1 Play)
        (:Update
          (Transformation|SetRelativeScale3D (Variables|Default|GetStaticMesh) (Math|Vector|Lerp(Vector) "0.750000,0.750000,0.750000" "1.000000,1.000000,1.000000" _lerp))
          (Utilities|FlowControl|Sequence
            (:then_0
              (if (Variables|Default|GetAlwaysOn)
                (Development|PrintString "BannerAlwaysOn")
                (else
                  (Utilities|FlowControl|Sequence
                    (:then_0
                      (Rendering|Components|Light|SetForceCachedShadowsforMovablePrimitives (Variables|Default|GetRimTopRight1) true))
                    (:then_1
                      ;[0 0 0 1 12] Force Disable on start
                      (CallFunction|SwitchMobileLighting false)
                      (CallFunction|SwitchPCLighting false)
                      (CallFunction|LightControl true))
                    (:then_2
                      ;[0 0 0 1 12] Sitch lighting if settings are switched while screen is already active
                      (bind _returnvalue_3 (Settings|GetGameUserSettings))
                      (EventDispatchers|BindEventtoOnGameUserSettingsUINeedsUpdate _returnvalue_3 (delegate Custom|UpdateSettings)))))))
            (:then_1
              (bind (_direction_2 _lerp_2 _rotator) (Timeline|Timeline_0 Play)
                (:Update
                  (Transformation|SetRelativeTransform (Variables|FortItemPreviewActor|GetItemPreviewRootComponent) (Math|Transform|MakeTransform (+ (.location (Transformation|GetRelativeTransform (Variables|FortItemPreviewActor|GetItemPreviewRootComponent))) (Math|Vector|Lerp(Vector) "0, 0,0.050000" "0.000000,0.000000,-0.050000" _lerp_2)) (.rotation (Transformation|GetRelativeTransform (Variables|FortItemPreviewActor|GetItemPreviewRootComponent))) (.scale (Transformation|GetRelativeTransform (Variables|FortItemPreviewActor|GetItemPreviewRootComponent)))) false false)
                  (Transformation|SetRelativeRotation (Variables|Default|GetStaticMesh) (Math|Rotator|Lerp(Rotator) "2.000000,10.000000,-5.000000" "-2.000000,-10.000000,5.000000" _rotator) false false))))))))
    (:then_1
      (Components|DestroyComponent (Variables|Default|GetObscuringLoopEmitter)))))

(event EventOnCurrentVisualsCleanedUp
  (Components|DestroyComponent (Variables|Default|GetObscuringLoopEmitter)))

(event Custom|UpdateSettings
  (bind _returnvalue (TouchPlatform|IsMobileGame))
  (bind _returnvalue_2 (TouchPlatform|IsNintendoSwitch))
  ;[0 0 0 1 12] Sitch lighting if settings are switched while screen is already active
  (if (Variables|Default|GetIsActive)
    (if (Math|Boolean|ORBoolean (== (Settings|GetShadowQuality (stale EventOnPreviewVisualsSpawned _returnvalue_3)) 0) _returnvalue _returnvalue_2)
      (if (or _returnvalue_2 _returnvalue)
        ;;[0.125 0.125 0.125 1 12] Enable Mobile
        (CallFunction|SwitchPCLighting false)
        (CallFunction|SwitchMobileLighting true)
        (else
          ;;[0.125 0.125 0.125 1 12] Enable Low Detail PC
          (CallFunction|SwitchPCLighting true)))
      (else
        ;;[0.125 0.125 0.125 1 12] Enable HighEnd
        (CallFunction|SwitchMobileLighting false)
        (CallFunction|SwitchPCLighting true)))))

(event Collision|EventActorBeginOverlap
  (params (OtherActor :type Actor%20Object%20Reference)))

(event EventTick
  (params (DeltaSeconds :type Float%20%28single-precision%29)))

(fn LightControl
  (params (Active :type Boolean))
  (bind _returnvalue (TouchPlatform|IsNintendoSwitch))
  (bind _returnvalue_2 (TouchPlatform|IsMobileGame))
  (bind _isactive (Variables|Default|GetIsActive))
  (if (not (Variables|Default|GetAlwaysOn))
    (Variables|Default|SetIsActive Active)
    (bind _returnvalue_3 (Settings|GetGameUserSettings))
    (if (Math|Boolean|ORBoolean (== (Settings|GetShadowQuality _returnvalue_3) 0) _returnvalue_2 _returnvalue)
      (if (or _returnvalue_2 _returnvalue)
        (CallFunction|SwitchMobileLighting _isactive)
        (elif _returnvalue
          (CallFunction|SwitchErebusLighting _isactive))))
    (CallFunction|SwitchPCLighting (Variables|Default|GetIsActive))))

(fn SwitchErebusLighting
  (params (Visibility :type Boolean))
  (Rendering|SetVisibility (Variables|Default|GetKeyLight_StandaloneForSwitch) Visibility false)
  (Rendering|Components|Light|SetCastShadows (Variables|Default|GetRimLeft1) false))

(fn SwitchMobileLighting
  (params (Visibility :type Boolean))
  (Rendering|SetVisibility (Variables|Default|GetKeyLight2) false false)
  (Rendering|SetVisibility (Variables|Default|GetDirectionalLightMobile) Visibility false))

(fn SwitchPCLighting
  (params (Visibility :type Boolean))
  (Rendering|SetVisibility (Variables|Default|GetRimLowerRight1) Visibility false))

(fn ConstructionScript)
