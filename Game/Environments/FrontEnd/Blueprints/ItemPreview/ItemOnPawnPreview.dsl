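; Front-end item preview actor that shows cosmetics on pawns, with its own lighting, floor and res-in presentation.
; On load it switches PC, mobile or Switch lights by platform and settings, and sets global toon-light material parameters.
; The floor hides for gliders and contrails and accepts custom or autotest materials.
; When shown, an optionally CVar-delayed res-in plays Niagara, sound and timeline-driven dissolves through temporary pawn material overrides, then mesh parameters.
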(event EventOnPreviewVisualsBeginLoading
  (Utilities|FlowControl|Sequence
    (:then_0
      (if (not (Variables|Default|GetAlwaysOn))
        (CallFunction|SetupLighting)
        (bind _returnvalue (Settings|GetGameUserSettings))
        (EventDispatchers|BindEventtoOnGameUserSettingsUINeedsUpdate _returnvalue (EventDispatchers|CreateEvent :!event HandleLightingSettingsChanged))))
    (:then_1
      ;[1 1 1] Setting Fake Light Vector For Toon Characters
      (Rendering|Material|SetVectorParameterValue "/Game/Packages/Fortress_SharedMaterials/GlobalMaterialParameters/FortniteMaterialParameters.FortniteMaterialParameters" "SunlightDirectionalVector" (Math|Conversions|ToLinearColor(Vector) (Math|Vector|RotateVector (Transformation|GetActorForwardVector) (Variables|Default|GetToonLightRotatio))) :!class /Script/Engine.KismetMaterialLibrary)
      (Rendering|Material|SetScalarParameterValue "/Game/Packages/Fortress_SharedMaterials/GlobalMaterialParameters/FortniteMaterialParameters.FortniteMaterialParameters" "bEnableCharacterRimLight" :!class /Script/Engine.KismetMaterialLibrary))))

(event EventOnPreviewVisualsSpawned
  (bind (_primaryitem _bshowfloor) (Utilities|Struct|BreakCharacterDisplaySettings (Variables|FortItemPreviewActor|GetCurrentDisplayInfo)))
  (bind _output_get (Variables|Default|SetShow%20Floor (and (Math|Boolean|NORBoolean (Utilities|IsA _primaryitem "/Script/FortniteGame.AthenaGliderItemDefinition") (Utilities|IsA _primaryitem "/Script/FortniteGame.AthenaSkyDiveContrailItemDefinition")) _bshowfloor)))
  (CallFunction|SetFloorEnabled _output_get))

(event EventOnPreviewVisualsShown
  (bind (_outvalue _returnvalue) (Fortnite|GetNumericCVar "FX.IASResinDelay"))
  (if (and (> _outvalue 0) _returnvalue)
    (bind _returnvalue_2 (Utilities|Time|SetTimerbyFunctionName 0 "SpawnResInEffects" _outvalue))
    (Variables|Default|SetIASResinDelayTimer _returnvalue_2)
    (else
      (CallFunction|SpawnResinEffects))))

(event EventOnCurrentVisualsCleanedUp
  (Utilities|Time|ClearandInvalidateTimerbyHandle (Variables|Default|GetIASResinDelayTimer)))

(event EventOnSetFloorMaterial
  (params (InMaterialInstance :type Material%20Interface%20Object%20Reference))
  ;[0 0 0] OnSetFloorMaterial
  (Variables|Default|SetCustomFloorMaterial InMaterialInstance)
  (Rendering|Material|SetMaterial (Variables|Default|GetPlane) 0 (select (Utilities|IsValid (Variables|Default|GetCustomFloorMaterial) :!class /Script/Engine.KismetSystemLibrary) (Variables|Default|GetCustomFloorMaterial) (Variables|Default|GetDefaultFloorMaterial))))

(event EventOnUpdateFloorMaterial
  (params (bEnableAutotestBackground :type Boolean))
  (if bEnableAutotestBackground
    (bind _returnvalue (Rendering|Material|CreateDynamicMaterialInstance (Variables|FortItemPreviewMultiAngleActor|GetAutotestFloorMaterial) "None" "None" :!class /Script/Engine.KismetMaterialLibrary))
    (Rendering|Material|SetMaterial (Variables|Default|GetPlane) 0 _returnvalue)
    (else
      (Rendering|Material|SetMaterial (Variables|Default|GetPlane) 0 (select (Utilities|IsValid (Variables|Default|GetCustomFloorMaterial) :!class /Script/Engine.KismetSystemLibrary) (Variables|Default|GetCustomFloorMaterial) (Variables|Default|GetDefaultFloorMaterial))))))

(event Custom|SpawnSoundPlayback
  ;[0 0 0] Resin sound playback
  (Audio|PlaySoundatLocation "/Game/Sounds/Fort_UI_Sounds/S13_Frontend/BattlePass_ResIn_Cue.BattlePass_ResIn_Cue" (CallFunction|GetPawnLocation)))

(event Custom|RezInMaterialEffect
  (bind _pawnmaterialoverrideguids (Variables|Default|GetPawnMaterialOverrideGuids))
  ;[1 1 1] RezInMaterialEffect
  (Utilities|Array|Clear (Variables|Default|GetPawnMaterialOverrideGuids))
  (bind _returnvalue (Rendering|Material|CreateDynamicMaterialInstance "/Game/Effects/Fort_Effects/UI/Transition_IAS/ChaDeRezShader/CharacterDeRezShader_frontend_Inst.CharacterDeRezShader_frontend_Inst" "None" "None" :!class /Script/Engine.KismetMaterialLibrary))
  (Rendering|Material|SetScalarParameterValue _returnvalue "DissolveGradientPercentage" 0.0 :!class /Script/Engine.MaterialInstanceDynamic)
  (for _array_element (CallFunction|GetPawns true)
    ;;[1 1 1] NEED OT FIX THIs. CHEcK IF RETURNS MULTIPLE GUIDS
    ;; IF SO - NEED TO STORE ALL OR FIND WAY TO REF
    (bind _returnvalue_2 (Class|FortPlayerPawn|ApplyMaterialOverrideLocalwithFlags :self _array_element :Material (Utilities|ResolveSoftReference _returnvalue) :Priority 0.0 :Flags 0)) ; need 
    (Utilities|Array|Add (Variables|Default|GetPawnMaterialOverrideGuids) _returnvalue_2))
  (bind (_direction _newtrack_0) (Timeline|RezInMaterialEffectTimeLine PlayFromStart)
    (:Update
      (Rendering|Material|SetScalarParameterValue _returnvalue "DissolveGradientPercentage" _newtrack_0 :!class /Script/Engine.MaterialInstanceDynamic)
      (for _array_element_2 (CallFunction|GetPawns true)
        (Utilities|SetPartModifierChildParticleComponentsVisibleonPawn _array_element_2)))
    (:Finished
      (Utilities|FlowControl|Sequence
        (:then_0
          (for (_array_element_3 _array_index) (CallFunction|GetPawns true)
            (if (Utilities|Array|IsValidIndex _pawnmaterialoverrideguids _array_index)
              (Class|FortPlayerPawn|RemoveMaterialOverrideLocal _array_element_3 (Utilities|Array|Get(aref) _pawnmaterialoverrideguids _array_index)))))
        (:then_1
          (bind (_direction_2 _newtrack_0_2) (Timeline|RezInMaterialEffectTimeLine2 PlayFromStart)
            (:Update
              (for _array_element_4 (CallFunction|GetPawns true)
                (bind _skinned_mesh_component (Utilities|ForEachSkinnedMeshComponentonPawn _array_element_4)
                  (:Loop%20Body
                    (CallFunction|UpdateRezinMaterialParameters _skinned_mesh_component _newtrack_0_2) ; Add all Cosmetic Skinned Meshes (but not the root skeletal mesh);\nthis already excludes any null components.
                  ))))
            (:Finished
              (for _array_element_5 (CallFunction|GetPawns true)
                ;[1 1 1] Add all Cosmetic Skinned Meshes (but not the root skeletal mesh); this already excludes any null components.
                (bind _skinned_mesh_component_2 (Utilities|ForEachSkinnedMeshComponentonPawn _array_element_5)
                  (:Loop%20Body
                    (Rendering|Material|SetScalarParameterValueonMaterials _skinned_mesh_component_2 "UsePannerAnimation" 0.0))
                  (:Loop%20Completed
                    (Utilities|SetPartModifierChildParticleComponentsVisibleonPawn _array_element_5 true)
                    (Class|FortPlayerPawn|RefreshWeaponBoneSetHiding _array_element_5)))))))))))

(fn GetLODStreamingSafetyDuration)

(fn HandleLightingSettingsChanged
  (bind _returnvalue (TouchPlatform|IsNintendoSwitch))
  (bind _returnvalue_2 (TouchPlatform|IsMobileGame))
  (bind _returnvalue_3 (Settings|GetFortGameUserSettings))
  (bind _returnvalue_4 (Class|FortGameUserSettings|IsinPerformanceMode _returnvalue_3))
  (bind _returnvalue_5 (Math|Boolean|ORBoolean _returnvalue_2 _returnvalue_4 _returnvalue))
  ;[0 0 0 1 12] Switch lighting if settings are changed while screen is already active
  (bind _returnvalue_6 (Settings|GetGameUserSettings))
  (if (Variables|Default|GetIsActive)
    (if (or (== (Settings|GetShadowQuality _returnvalue_6) 0) _returnvalue_5)
      (if _returnvalue_5
        ;;[0.125 0.125 0.125 1 12] Enable Mobile
        (CallFunction|SwitchPCLighting false)
        (CallFunction|SwitchMobileLighting true)
        (else
          ;;[0.125 0.125 0.125 1 12] Enable Low Detail PC
          (CallFunction|SwitchPCLighting true)))
      (else
        ;;[0.125 0.125 0.125 1 12] Enable HighEnd
        (CallFunction|SwitchMobileLighting false)
        (CallFunction|SwitchPCLighting true)))))

(fn IsSkyDiveContrailItem
  (bind (_asathena_sky_dive_contrail_item_definition _bsuccess) (Utilities|Casting|CastToAthenaSkyDiveContrailItemDefinition (Utilities|Struct|BreakCharacterDisplaySettings (Variables|FortItemPreviewActor|GetCurrentDisplayInfo))))
  (return _bsuccess))

(fn LightControl
  (params (Active :type Boolean))
  (bind _returnvalue (Settings|GetFortGameUserSettings))
  (bind _returnvalue_2 (Class|FortGameUserSettings|IsinPerformanceMode _returnvalue))
  (bind _returnvalue_3 (TouchPlatform|IsNintendoSwitch))
  (bind _returnvalue_4 (TouchPlatform|IsNintendoSwitch))
  (bind _returnvalue_5 (TouchPlatform|IsMobileGame))
  (bind _returnvalue_6 (Math|Boolean|ORBoolean _returnvalue_5 _returnvalue_2 _returnvalue_3))
  (bind _isactive (Variables|Default|GetIsActive))
  (if (not (Variables|Default|GetAlwaysOn))
    (Variables|Default|SetIsActive Active)
    (bind _returnvalue_7 (Settings|GetGameUserSettings))
    (if (Math|Boolean|ORBoolean (== (Settings|GetShadowQuality _returnvalue_7) 0) _returnvalue_6 _returnvalue_4)
      (if _returnvalue_6
        (CallFunction|SwitchMobileLighting _isactive)
        (elif _returnvalue_4
          (CallFunction|SwitchErebusLighting _isactive))))
    (CallFunction|SwitchPCLighting (Variables|Default|GetIsActive))))

(fn SetFloorEnabled
  (params (Show%20Floor :type Boolean))
  (Development|SetHiddeninGame (Variables|Default|GetPlane) (not Show%20Floor) true))

(fn SetupLighting
  (if (not (Variables|Default|GetAlwaysOn))
    (Utilities|FlowControl|Sequence
      (:then_0
        (Rendering|Components|Light|SetForceCachedShadowsforMovablePrimitives (Variables|Default|GetKeyLight5) true))
      (:then_1
        ;[0 0 0 1 12] Forece Disable on start
        (CallFunction|SwitchMobileLighting false)
        (CallFunction|SwitchPCLighting false)
        (CallFunction|LightControl (not (CallFunction|ShouldDisableLights)))))))

(fn SpawnResInEffects
  (for _array_element (CallFunction|GetPawnLocations)
    (Components|DestroyComponent (Variables|Default|GetObscuringLoopEmitter))
    (Niagara|SpawnSystematLocation "/Game/Effects/Fort_Effects/UI/Transition/P_FrontEndCharacterTransition_Converted.P_FrontEndCharacterTransition_Converted" (- _array_element "0.000000,0.000000,80.000000")))
  (CallFunction|RezinMaterialEffect)
  (CallFunction|SpawnSoundPlayback) ; Just play a single sound regardless of how many pawns
)

(fn SwitchErebusLighting
  (params (Visibility :type Boolean))
  (Rendering|SetVisibility (Variables|Default|GetKeyLight_StandaloneForSwitch) Visibility false))

(fn SwitchMobileLighting
  (params (NewParam :type Boolean))
  (Rendering|SetVisibility (Variables|Default|GetBounce1) false false)
  (Rendering|SetVisibility (Variables|Default|GetDirectionalLightMobile) NewParam false))

(fn SwitchPCLighting
  (params (Visibility :type Boolean))
  (Rendering|SetVisibility (Variables|Default|GetBounce1) Visibility false)
  (Rendering|SetVisibility (Variables|Default|GetKeyLight_StandaloneForSwitch) false false))

(fn UpdateRezInMaterialParameters
  (params (SkinnedMesh :type Skinned%20Mesh%20Component%20Object%20Reference) (DissolveGradientValue :type Float%20%28single-precision%29))
  (Rendering|Material|SetScalarParameterValueonMaterials SkinnedMesh "NewDissolveGradient" DissolveGradientValue)
  (Rendering|Material|SetScalarParameterValueonMaterials SkinnedMesh "UsePannerAnimation" 1.0))

(fn ConstructionScript)
