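; Front-end item preview actor for contrails shown on a pawn, which sets directional wind strength based on whether the preview pawn has cloth.
; When preview visuals spawn, it waits briefly for loading, then caches the wind components from its child wind source actor.
; It checks the pawn's skeletal meshes for a Chaos Cloth Asset simulation interactor to flag cloth.
; Wind is set to a gentle strength with cloth, or zero without.
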(event EventOnPreviewVisualsSpawned
  (Utilities|FlowControl|Delay 0.3) ; Because apparantly this fire too quickly and if we dont delay stuff wont be loaded.
  (Variables|Default|SetWindDirComponents (Actor|GetComponentsByClass (Variables|/Script/Engine.ChildActorComponent|GetChildActor (Variables|Default|GetWindDirectionalSource)) "/Script/Engine.WindDirectionalSourceComponent"))
  (Utilities|FlowControl|Sequence
    (:then_0
      (for _array_element (Actor|GetComponentsByClass (Variables|/Script/Engine.ChildActorComponent|GetChildActor (Variables|FortItemPreviewOnPawnActor|GetPrimaryPreviewPawnComponent)) "/Script/Engine.SkeletalMeshComponent")
        (Utilities|Casting|CastToChaosClothAssetSKMClothingSimulationInteractor (Clothing|GetClothingSimulationInteractor _array_element "None"))
        (Variables|Default|SetbHasCloth true)))
    (:then_1
      (if (Variables|Default|GetbHasCloth)
        (for _array_element_2 (Variables|Default|GetWindDirComponents)
          (Variables|/Script/Engine.WindDirectionalSourceComponent|SetStrength 0.1 _array_element_2))
        (else
          (for _array_element_3 (Variables|Default|GetWindDirComponents)
            (Variables|/Script/Engine.WindDirectionalSourceComponent|SetStrength 0.0 _array_element_3)))))))

(fn ConstructionScript
  (|CallParentFunction))
