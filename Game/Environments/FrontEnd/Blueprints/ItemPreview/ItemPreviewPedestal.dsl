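; Front-end item preview pedestal Blueprint, a camera target pedestal that can suppress the floor of the preview it displays.
; When an item is displayed and its Floor Enabled flag is off, it passes the current item preview actor to RemoveFloor.
; RemoveFloor tries casts to ItemOnPawnPreview and WrapPreview in sequence, disabling the floor on each.
; Begin play, overlap, tick and the construction script are empty.
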(event EventBeginPlay)

(event Collision|EventActorBeginOverlap
  (params (OtherActor :type Actor%20Object%20Reference)))

(event EventTick
  (params (DeltaSeconds :type Float%20%28single-precision%29)))

(event EventOnItemDisplayed
  (params (ItemType :type EFortItemType%20Enum :default "WorldItem"))
  (if (not (Variables|Default|GetFloor%20Enabled))
    (CallFunction|RemoveFloor (Variables|FortCameraTargetPedestal|GetItemPreviewActor))))

(fn RemoveFloor
  (params (Preview :type Fort%20Item%20Preview%20Actor%20Object%20Reference))
  (Utilities|FlowControl|Sequence
    (:then_0
      (bind _asitem_on_pawn_preview (Utilities|Casting|CastToItemOnPawnPreview Preview))
      (Class|ItemonPawnPreview|SetFloorEnabled _asitem_on_pawn_preview false))
    (:then_1
      (bind _aswrap_preview (Utilities|Casting|CastToWrapPreview Preview))
      (Class|WrapPreview|SetFloorEnabled _aswrap_preview false))))

(fn ConstructionScript)
