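; Front-end item preview scene changer that swaps previewed items by sliding scenes sideways, clockwise or counterclockwise.
; Timelines move the previous scene out and the new one in before signaling SwapComplete, and also ease zoom and pedestal rotation toward their targets.
; LightControl picks PC, low-detail PC or mobile lighting from platform, shadow quality and performance mode.
; Unless AlwaysOn, lighting starts disabled and re-evaluates on settings changes while active.
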(event EventBeginPlay
  (if (Variables|Default|GetAlwaysOn)
    (Development|PrintString "ALWAYS ON")
    (else
      (Utilities|FlowControl|Sequence
        (:then_0
          ;[0 0 0 1 12] Force Disable On Start Up
          (CallFunction|SwitchPCLighting false)
          (CallFunction|SwitchMobileLighting false)
          (CallFunction|SwitchPCLightingLOWDetail false))
        (:then_1
          ;[0 0 0 1 12] Switch lighting if settings are update while screen is active
          (bind _returnvalue (Settings|GetGameUserSettings))
          (EventDispatchers|BindEventtoOnGameUserSettingsUINeedsUpdate _returnvalue (delegate Custom|UpdateSettings)))))))

(event Custom|UpdateSettings
  (bind _returnvalue (TouchPlatform|IsMobileGame))
  (bind _returnvalue_2 (Settings|GetFortGameUserSettings))
  (bind _returnvalue_3 (Class|FortGameUserSettings|IsinPerformanceMode _returnvalue_2))
  (bind _returnvalue_4 (or _returnvalue _returnvalue_3))
  ;[0 0 0 1 12] Switch lighting if settings are update while screen is active
  (if (Variables|Default|GetIsActive)
    (if (or (== (Settings|GetShadowQuality (stale EventBeginPlay _returnvalue)) 0) _returnvalue_4)
      (if _returnvalue_4
        ;;[0.125 0.125 0.125 1 12] Force Mobile
        (CallFunction|SwitchPCLighting false)
        (CallFunction|SwitchPCLightingLOWDetail false)
        (CallFunction|SwitchMobileLighting true)
        (Development|PrintString "MobileOnDesktop: Forced Mobile Lighting")
        (else
          ;;[0.125 0.125 0.125 1 12] Force PC Low Detail
          (CallFunction|SwitchPCLighting false)
          (CallFunction|SwitchMobileLighting false)
          (CallFunction|SwitchPCLightingLOWDetail true)
          (Development|PrintString "MobileOnDesktop: Forced Low-PC Lighting")))
      (else
        ;;[0.125 0.125 0.125 1 12] Force PC Lighting
        (CallFunction|SwitchMobileLighting false)
        (CallFunction|SwitchPCLightingLOWDetail false)
        (CallFunction|SwitchPCLighting true)
        (Development|PrintString "MobileOnDesktop: Forced PC Lighting")))))

(event EventOnNewSceneRequested
  (params (PrimaryRequestedItem :type Fort%20Account%20Item%20Definition%20Object%20Reference) (TransitionOptions :type Scene%20Transition%20Options%20Structure%20%28by%20ref%29))
  (Variables|Default|SetPrimary%20Requested%20Item PrimaryRequestedItem)
  (switch Utilities|FlowControl|Switch|SwitchonESceneTransitionType (Utilities|Struct|BreakSceneTransitionOptions TransitionOptions)
    (:Clockwise
      (Utilities|FlowControl|Sequence
        (:then_0
          (goto _label Reset))
        (:then_1
          (bind (_direction _location) (Timeline|MoveOutLeft PlayFromStart)
            (:Update
              (CallFunction|UpdatePreviousSceneTransform (Math|Transform|MakeTransform _location)))
            (:Finished
              _))))
      (Utilities|FlowControl|WaitforBothExecutions true :!label _label
        (:Then
          (Variables|Default|SetPrevious%20Requested%20Item (Variables|Default|GetPrimary%20Requested%20Item))
          (bind (_direction_2 _location_2) (Timeline|MoveInRight PlayFromStart)
            (:Update
              (CallFunction|UpdateNewSceneTransform (Math|Transform|MakeTransform _location_2)))
            (:Finished
              (CallFunction|SwapComplete))))))
    (:CounterClockwise
      (Utilities|FlowControl|Sequence
        (:then_0
          (goto _label_2 Reset))
        (:then_1
          (bind (_direction_3 _location_3) (Timeline|MoveOutRight PlayFromStart)
            (:Update
              (CallFunction|UpdatePreviousSceneTransform (Math|Transform|MakeTransform _location_3)))
            (:Finished
              _))))
      (Utilities|FlowControl|WaitforBothExecutions true :!label _label_2
        (:Then
          (bind (_direction_4 _location_4) (Timeline|MoveInLeft PlayFromStart)
            (:Update
              (CallFunction|UpdateNewSceneTransform (Math|Transform|MakeTransform _location_4)))
            (:Finished
              (CallFunction|SwapComplete))))))))

(event EventOnReadyToSwap
  (params (PrimaryRequestedItem :type Fort%20Account%20Item%20Definition%20Object%20Reference) (Options :type Scene%20Transition%20Options%20Structure%20%28by%20ref%29))
  (Utilities|FlowControl|Sequence
    (:then_0
      (Utilities|FlowControl|Delay 0.1)
      (switch Utilities|FlowControl|Switch|SwitchonESceneTransitionType (Utilities|Struct|BreakSceneTransitionOptions Options)
        (:NoTransition
          (CallFunction|SwapComplete))
        (:Clockwise
          (CallFunction|UpdateNewSceneTransform (Math|Transform|MakeTransform "2000.000000, 0, 0"))
          (Utilities|FlowControl|WaitforBothExecutions true
            (:Then
              (Variables|Default|SetPrevious%20Requested%20Item (Variables|Default|GetPrimary%20Requested%20Item))
              (Timeline|MoveInRight PlayFromStart))))
        (:CounterClockwise
          (CallFunction|UpdateNewSceneTransform (Math|Transform|MakeTransform "-2000.000000, 0, 0"))
          (Utilities|FlowControl|WaitforBothExecutions true
            (:Then
              (Timeline|MoveInLeft PlayFromStart))))))
    (:then_1
      (Utilities|IsValid PrimaryRequestedItem
        (:Is%20Valid
          (Default|CallOnNewItemSpawned (not (or (Utilities|Enum|Equal(Enum) (Item|GetItemType PrimaryRequestedItem) "AthenaSkyDiveContrail") (or (Utilities|Enum|Equal(Enum) (Item|GetItemType PrimaryRequestedItem) "AthenaGlider") (Utilities|Enum|Equal(Enum) (Item|GetItemType PrimaryRequestedItem) "Token"))))))))))

(event EventOnTargetZoomLevelSet
  (params (TargetZoomLevel :type Float%20%28single-precision%29))
  (bind _output_get (Variables|Default|SetTargetZoomLevel TargetZoomLevel))
  (Math|Float|CompareFloat _output_get (Variables|FortItemPreviewSceneChanger|GetZoomLevel)
    (:>
      (bind (_direction _zoomlevel) (Timeline|Timeline_0 Play)
        (:Update
          (Utilities|FlowControl|Sequence
            (:then_0
              (switch Utilities|FlowControl|Switch|SwitchonETimelineDirection _direction
                (:Forward
                  (if (not (>= _zoomlevel (Variables|Default|GetTargetZoomLevel)))
                    (Variables|FortItemPreviewSceneChanger|SetZoomLevel _zoomlevel :!label _label)))
                (:Backward
                  (if (not (<= _zoomlevel (Variables|Default|GetTargetZoomLevel)))
                    (goto _label))))
              (Components|Timeline|Stop (Variables|ItemPreviewSideSwap|GetTimeline_0))
              (Variables|Default|SetTargetZoomLevel (Variables|Default|GetTargetZoomLevel)))
            (:then_1
              _)))
        (:Finished
          (Variables|FortItemPreviewSceneChanger|SetZoomLevel (Variables|Default|GetTargetZoomLevel)))))
    (:==
      _)
    (:<
      (Timeline|Timeline_0 Reverse)))
  (CallFunction|TickZoomLevel))

(event EventOnTargetRotationOffsetSet
  (params (TargetRotationOffset :type Rotator%20%28by%20ref%29 :default "0, 0, 0"))
  (bind _output_get (Variables|Default|SetTargetRotationOffset TargetRotationOffset))
  (bind _output_get_2 (Variables|Default|SetOrigRotationOffset (Variables|FortItemPreviewSceneChanger|GetPedestalRotationOffset)))
  (if (not (Math|Rotator|Equal(Rotator) _output_get_2 _output_get 0.0001))
    (bind (_direction _rotationoffsetlerp) (Timeline|Timeline_1 Play)
      (:Update
        (Variables|FortItemPreviewSceneChanger|SetPedestalRotationOffset (Math|Rotator|Lerp(Rotator) (Variables|Default|GetOrigRotationOffset) (Variables|Default|GetTargetRotationOffset) _rotationoffsetlerp)))
      (:Finished
        (Variables|FortItemPreviewSceneChanger|SetPedestalRotationOffset (Variables|Default|GetTargetRotationOffset)))))
  (CallFunction|TickPedestalRotationOffset))

(fn LightControl
  (params (Active :type Boolean))
  (bind _returnvalue (Settings|GetFortGameUserSettings))
  (bind _returnvalue_2 (Class|FortGameUserSettings|IsinPerformanceMode _returnvalue))
  (bind _returnvalue_3 (TouchPlatform|IsMobileGame))
  (bind _returnvalue_4 (or _returnvalue_3 _returnvalue_2))
  (if (not (Variables|Default|GetAlwaysOn))
    (Variables|Default|SetIsActive Active)
    (bind _returnvalue_5 (Settings|GetGameUserSettings))
    (if (Math|Boolean|ORBoolean (== (Settings|GetShadowQuality _returnvalue_5) 0) (TouchPlatform|IsNintendoSwitch) _returnvalue_4)
      (if (or (TouchPlatform|IsNintendoSwitch) _returnvalue_4)
        (CallFunction|SwitchMobileLighting Active)
        (Development|PrintString "MobileOnDesktop: Forced Mobile Lighting 2")
        (else
          (CallFunction|SwitchPCLightingLOWDetail Active)
          (Development|PrintString "MobileOnDesktop: Forced Low-PC Lighting 2")))
      (else
        (CallFunction|SwitchPCLighting Active)
        (Development|PrintString "MobileOnDesktop: Forced PC Lighting 2")))))

(fn SwitchMobileLighting
  (params (Visibility :type Boolean))
  (bind _postprocess_mobile (Variables|Default|GetPostProcess_Mobile))
  (Rendering|SetVisibility (Variables|Default|GetSkyLight) false false)
  (Rendering|SetVisibility (Variables|Default|GetDirectionalLightMobile) (Variables|GetVisibility) false)
  (Utilities|IsValid _postprocess_mobile
    (:Is%20Valid
      (Variables|/Script/Engine.PostProcessComponent|SetbEnabled (Variables|GetVisibility) _postprocess_mobile))))

(fn SwitchPCLighting
  (params (Visibility :type Boolean))
  (Rendering|SetVisibility (Variables|Default|GetSkyLight) Visibility false))

(fn SwitchPCLighting_LOWDetail
  (params (Visibility :type Boolean))
  (bind _postprocess_lowpc (Variables|Default|GetPostProcess_LOWPC))
  (Rendering|SetVisibility (Variables|Default|GetDirectionalLight_LOWPC) (Variables|GetVisibility) false)
  (Utilities|IsValid _postprocess_lowpc
    (:Is%20Valid
      (Variables|/Script/Engine.PostProcessComponent|SetbEnabled (Variables|GetVisibility) _postprocess_lowpc))))

(fn ConstructionScript)
