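# Masked, lit material for previewing a Battle Pass icon in the front end, with a shine swipe and a gentle bobbing motion.
# The center-scaled texture's dimmed RGB is added to a shine swipe weighted by its red channel.
# A time-driven sine blends bob height vectors and a slight Z-axis rotation about the object's position, forming an apparent world position offset.
# Texture alpha is clamped to a transparent border for masking.
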Material /Game/Environments/FrontEnd/Blueprints/ItemPreview/Materials/M-BP-BattlePassPreviewMaterial
graph {
 properties {blendMode=BLEND_Masked, pixelDepthOffsetMode=PDOM_Legacy, usedShadingModels=Default Lit}
 nodes [
  n1 = MaterialFunctionCall(Texture Scale=ScalarParameter() {defaultValue=0.85, parameterName=Scale}) {materialFunction=&ScaleUVsByCenter}
  n2 = TextureSampleParameter2D(UVs=@n1.UVs) {parameterName=2d-Texture, texture=/Game/UI/Foundation/Textures/Icons/Items/T-FNBR-BattlePass-L}
  n3 = ScalarParameter() {defaultValue=4, parameterName=Snie-Multiply}
  n4 = MaterialFunctionCall!(In_Uv) {materialFunction=/Game/UI/Foundation/Materials/M-Function-ShineSwipe}
  Add(
   A=Multiply(A=@n2.RGB, B=ScalarParameter() {defaultValue=0.35, parameterName=RGB-Multiply})
   B=Multiply(A=@n2.R, B=Multiply(A=@n3, B=@n4.Alpha))
  )
  n5 = Sine(None=Multiply(A=Time(), B=ScalarParameter() {defaultValue=0.2, parameterName=BobRate}))
  n6 = Constant() {r=-0.01}
  n7 = Constant() {r=0.01}
  Add(
   A=LinearInterpolate(
    A=VectorParameter() {parameterName=BobHeightMax, defaultValue=<rgba 0 0 5 0>}.RGB
    B=VectorParameter() {parameterName=BobHeightMin}.RGB
    Alpha=@n5
   )
   B=RotateAboutAxis(
    NormalizedRotationAxis=Constant3Vector() {constant=<rgba 0 0 1 0>}
    RotationAngle=LinearInterpolate(A=@n6, B=@n7, Alpha=@n5)
    PivotPoint=ObjectPositionWS()
    Position=WorldPosition().XYZ
   )
  )
  Constant()
  Reroute(None=MaterialFunctionCall!(UVs=@n1.UVs, Alpha=@n2.A) {materialFunction=/Game/UI/Foundation/Items/ClampAlphaToTransparentBorder}.Result)
 ]
}
