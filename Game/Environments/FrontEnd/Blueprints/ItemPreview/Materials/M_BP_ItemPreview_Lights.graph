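# Translucent two-sided material for item preview lights in the front end, drawing a procedural, tintable light ray with adjustable width, intensity and opacity.
# A looping ping-pong time value, eased in and out, rotates UVs to sway the ray.
# The ray mask combines a rounded tapering profile with a reflected gradient along V.
# Taper width varies with the vertex normal's alignment to the instance Z axis.
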Material /Game/Environments/FrontEnd/Blueprints/ItemPreview/Materials/M_BP_ItemPreview_Lights
graph {
 properties {blendMode=BLEND_Translucent, pixelDepthOffsetMode=PDOM_Legacy, twoSided=true, usedShadingModels=Default Lit}
 nodes [
  Multiply(A=NamedRerouteUsage(), B=ScalarParameter() {defaultValue=1, group=Ray, parameterName=Opacity})
  n1 = VectorParameter() {group=Color, parameterName=Color, sortPriority=1, defaultValue=<rgba 1 1 1 1>}
  NamedRerouteDeclaration(None=Multiply(A=@n1.RGB, B=ScalarParameter() {defaultValue=1, group=Color, parameterName=ColorIntensity, sortPriority=2})) {name=Color, nodeColor=<rgba 1 0 1 1>}
  n2 = TextureCoordinate()
  n3 = Time()
  n4 = ScalarParameter() {defaultValue=0.2, group=Animation, parameterName=AnimationSpeed, sortPriority=1}
  n5 = Multiply(A=Frac(None=Multiply(A=@n3, B=@n4) {constB=0.2})) {constB=2}
  n6 = MaterialFunctionCall("Value (0-1)"=Abs(None=Subtract(A=@n5) {constB=2})) {materialFunction=&MF_UI_EaseCurves}
  NamedRerouteDeclaration(None=Panner(Coordinate=Rotator(Coordinate=@n2, Time=@n6.EaseInOutSine) {centerY=0, speed=0.5}, Time=Time())) {name=Animation, nodeColor=<rgba 0 0 0 1>}
  NamedRerouteDeclaration(None=@n1.A) {name=Color Alpha, nodeColor=<rgba 0 0 0 1>}
  n7 = Add(A=NamedRerouteUsage()) {constB=0.1}
  n8 = ComponentMask(None=TextureCoordinate()) {g=true}
  n9 = OneMinus(None=@n8)
  n10 = ScalarParameter() {defaultValue=0.2, group=Ray, parameterName=BaseRayWidth, sortPriority=2}
  n11 = Constant() {r=0.4}
  n12 = VertexNormalWS()
  n13 = Transform(None=Constant3Vector() {constant=<rgba 0 0 1 0>}) {transformSourceType=TRANSFORMSOURCE_Instance}
  n14 = Power(Base=Abs(None=DotProduct(A=@n12, B=@n13)), Exp=Constant() {r=2})
  n15 = OneMinus(None=Saturate(None=LinearInterpolate(A=@n10, B=@n11, Alpha=@n14)))
  n16 = Divide(A=ConstantBiasScale(None=ComponentMask(None=@n7) {r=true}) {bias=-0.5, scale=2}, B=OneMinus(None=Multiply(A=@n9, B=@n15)))
  n17 = OneMinus(None=Saturate(None=DotProduct(A=@n16, B=@n16)))
  n18 = Power(Base=Saturate(None=SquareRoot(None=@n17)), Exp=ScalarParameter() {defaultValue=2, group=Ray, parameterName=RayWidth, sortPriority=1})
  n19 = MaterialFunctionCall!(Gradient=@n8, Center Point=Constant() {r=0.005}) {materialFunction=/Game/Effects/Fort_Effects/MaterialFunctions/MF_GradientReflected}
  NamedRerouteDeclaration(None=Multiply(A=Multiply(A=@n18, B=@n19.Result), B=NamedRerouteUsage())) {name=Mask, nodeColor=<rgba 1 0 1 1>}
  NamedRerouteUsage()
 ]
}
