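; Front-end preview actor for music pack items that animates its placeholder mesh and manages its preview lighting.
; On spawn, a timeline scales the mesh up; when it finishes, non-AlwaysOn previews enable cached shadows, initialize lights, bind to settings changes and play a reveal sound.
; LightControl and UpdateSettings toggle a PC rim light and a mobile directional light based on platform, performance mode and shadow quality.
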(event Collision|EventActorBeginOverlap
  (params (OtherActor :type Actor%20Object%20Reference)))

(event EventTick
  (params (DeltaSeconds :type Float%20%28single-precision%29)))

(event EventOnPreviewVisualsSpawned
  (bind (_direction _lerp) (Timeline|Timeline_0 Play)
    (:Update
      (Transformation|SetRelativeScale3D (Variables|Default|GetPlaceholderMesh) (Math|Vector|Lerp(Vector) "0.850000,0.850000,0.850000" "1.000000,1.000000,1.000000" _lerp)))
    (:Finished
      (if (Variables|Default|GetAlwaysOn)
        (Development|PrintString "BannerAlwaysOn")
        (else
          (Utilities|FlowControl|Sequence
            (:then_0
              (Rendering|Components|Light|SetForceCachedShadowsforMovablePrimitives (Variables|Default|GetRimTopRight1) true))
            (:then_1
              ;[0 0 0 1 12] Force Disable on start
              (CallFunction|SwitchMobileLighting false)
              (CallFunction|SwitchPCLighting false)
              (CallFunction|LightControl true))
            (:then_2
              ;[0 0 0 1 12] Sitch lighting if settings are switched while screen is already active
              (bind _returnvalue (Settings|GetGameUserSettings))
              (EventDispatchers|BindEventtoOnGameUserSettingsUINeedsUpdate _returnvalue (delegate Custom|UpdateSettings)))
            (:then_3
              (Audio|PlaySoundatLocation "/Game/Sounds/Fort_UI_Sounds/S13_Frontend/BattlePass_ResIn_Cue.BattlePass_ResIn_Cue" (Transformation|GetActorLocation)))))))))

(event Custom|UpdateSettings
  (bind _returnvalue (Settings|GetFortGameUserSettings))
  (bind _returnvalue_2 (Class|FortGameUserSettings|IsinPerformanceMode _returnvalue))
  (bind _returnvalue_3 (or false _returnvalue_2))
  ;[0 0 0 1 12] Sitch lighting if settings are switched while screen is already active
  (if (Variables|Default|GetIsActive)
    (if (or (== (Settings|GetShadowQuality (stale EventOnPreviewVisualsSpawned _returnvalue)) 0) _returnvalue_3)
      (if _returnvalue_3
        ;;[0.125 0.125 0.125 1 12] Enable Mobile
        (CallFunction|SwitchPCLighting false)
        (CallFunction|SwitchMobileLighting true)
        (else
          ;;[0.125 0.125 0.125 1 12] Enable Low Detail PC
          (CallFunction|SwitchPCLighting true)))
      (else
        ;;[0.125 0.125 0.125 1 12] Enable HighEnd
        (CallFunction|SwitchMobileLighting false)
        (CallFunction|SwitchPCLighting true)))))

(fn LightControl
  (params (Active :type Boolean))
  (bind _returnvalue (Settings|GetFortGameUserSettings))
  (bind _returnvalue_2 (Class|FortGameUserSettings|IsinPerformanceMode _returnvalue))
  (bind _returnvalue_3 (TouchPlatform|IsNintendoSwitch))
  (bind _returnvalue_4 (TouchPlatform|IsMobileGame))
  (bind _returnvalue_5 (Math|Boolean|ORBoolean _returnvalue_4 _returnvalue_2 _returnvalue_3))
  (if (not (Variables|Default|GetAlwaysOn))
    (Variables|Default|SetIsActive Active)
    (bind _returnvalue_6 (Settings|GetGameUserSettings))
    (if (Math|Boolean|ORBoolean (== (Settings|GetShadowQuality _returnvalue_6) 0) _returnvalue_5 (TouchPlatform|IsNintendoSwitch))
      (if _returnvalue_5
        (CallFunction|SwitchMobileLighting Active)
        (else
          _)))
    (CallFunction|SwitchPCLighting Active)))

(fn SwitchMobileLighting
  (params (Visibility :type Boolean))
  (Rendering|SetVisibility (Variables|Default|GetRimLowerRight1) false false)
  (Rendering|SetVisibility (Variables|Default|GetDirectionalLightMobile) Visibility false))

(fn SwitchPCLighting
  (params (Visibility :type Boolean))
  (Rendering|SetVisibility (Variables|Default|GetRimLowerRight1) Visibility false))

(fn ConstructionScript)
