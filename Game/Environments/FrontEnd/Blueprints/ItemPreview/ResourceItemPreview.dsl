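; Front-end resource item preview actor that plays a reveal transition and manages its lighting rig.
; On spawn it hides the preview, plays transition particles and a sound, then after a short delay reveals it and activates lighting.
; LightControl selects mobile, Switch or PC lights from platform, performance mode and shadow quality.
; A bound user-settings event re-switches lighting while active, and shown or cleanup events destroy the looping obscuring emitter.
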(event EventOnPreviewVisualsSpawned
  (bind _previewactorcomponent (Variables|FortItemPreviewOffPawnActor|GetPreviewActorComponent))
  (bind _returnvalue (Transformation|GetWorldLocation _previewactorcomponent))
  (Rendering|SetVisibility (Variables|FortItemPreviewOffPawnActor|GetPreviewActorComponent) false false)
  (Components|DestroyComponent (Variables|Default|GetObscuringLoopEmitter))
  (Effects|SpawnEmitteratLocation :EmitterTemplate "/Game/Effects/Fort_Effects/UI/Transition/P_FrontEndCharacterTransition.P_FrontEndCharacterTransition" :Location _returnvalue :Scale "1.125000,1.125000,1.125000" :bAutoDestroy true :PoolingMethod "None" :bAutoActivateSystem true)
  (bind _returnvalue_2 (Effects|SpawnEmitteratLocation "/Game/Effects/Fort_Effects/UI/Transition/P_FrontEndCharacterTransition_Loop.P_FrontEndCharacterTransition_Loop" _returnvalue))
  (Variables|Default|SetObscuringLoopEmitter _returnvalue_2)
  (Audio|PlaySoundatLocation "/Game/Sounds/Fort_UI_Sounds/S13_Frontend/BattlePass_ResIn_Cue.BattlePass_ResIn_Cue")
  (Utilities|FlowControl|Delay 0.2)
  (Components|DestroyComponent (Variables|Default|GetObscuringLoopEmitter))
  (Rendering|SetVisibility (Variables|FortItemPreviewOffPawnActor|GetPreviewActorComponent) true false)
  (Utilities|FlowControl|Sequence
    (:then_0
      (Rendering|Components|Light|SetForceCachedShadowsforMovablePrimitives (Variables|Default|GetRimLowerRight1) true))
    (:then_1
      ;[0 0 0 1 12] Force Disable on start
      (CallFunction|SwitchMobileLighting false)
      (CallFunction|SwitchPCLighting false)
      (CallFunction|LightControl true))
    (:then_2
      ;[0 0 0 1 12] Sitch lighting if settings are switched while screen is already active
      (bind _returnvalue_3 (Settings|GetGameUserSettings))
      (EventDispatchers|BindEventtoOnGameUserSettingsUINeedsUpdate _returnvalue_3 (delegate Custom|UpdateSettings)))))

(event Custom|UpdateSettings
  (bind _returnvalue (TouchPlatform|IsMobileGame))
  (bind _returnvalue_2 (Settings|GetFortGameUserSettings))
  (bind _returnvalue_3 (Class|FortGameUserSettings|IsinPerformanceMode _returnvalue_2))
  (bind _returnvalue_4 (or _returnvalue _returnvalue_3))
  ;[0 0 0 1 12] Sitch lighting if settings are switched while screen is already active
  (if (Variables|Default|GetIsActive)
    (if (or (== (Settings|GetShadowQuality (stale EventOnPreviewVisualsSpawned _returnvalue_3)) 0) _returnvalue_4)
      (if _returnvalue_4
        ;;[0.125 0.125 0.125 1 12] Enable Mobile
        (CallFunction|SwitchPCLighting false)
        (CallFunction|SwitchMobileLighting true)
        (else
          ;;[0.125 0.125 0.125 1 12] Enable Low Detail PC
          (CallFunction|SwitchPCLighting true)))
      (else
        ;;[0.125 0.125 0.125 1 12] Enable HighEnd
        (CallFunction|SwitchMobileLighting false)
        (CallFunction|SwitchPCLighting true)))))

(event EventOnPreviewVisualsShown
  (Components|DestroyComponent (Variables|Default|GetObscuringLoopEmitter)))

(event EventOnCurrentVisualsCleanedUp
  (Components|DestroyComponent (Variables|Default|GetObscuringLoopEmitter)))

(fn LightControl
  (params (Active :type Boolean))
  (bind _returnvalue (TouchPlatform|IsNintendoSwitch))
  (bind _returnvalue_2 (Settings|GetFortGameUserSettings))
  (bind _returnvalue_3 (Class|FortGameUserSettings|IsinPerformanceMode _returnvalue_2))
  (bind _returnvalue_4 (TouchPlatform|IsMobileGame))
  (bind _returnvalue_5 (TouchPlatform|IsNintendoSwitch))
  (bind _returnvalue_6 (Math|Boolean|ORBoolean _returnvalue_4 _returnvalue_3 _returnvalue))
  (bind _isactive (Variables|Default|GetIsActive))
  (Variables|Default|SetIsActive Active)
  (bind _returnvalue_7 (Settings|GetGameUserSettings))
  (if (Math|Boolean|ORBoolean (== (Settings|GetShadowQuality _returnvalue_7) 0) _returnvalue_6 _returnvalue_5)
    (if _returnvalue_6
      (CallFunction|SwitchMobileLighting _isactive)
      (else
        _)))
  (if _returnvalue_5
    (CallFunction|SwitchErebusLighting _isactive)
    (else
      (CallFunction|SwitchPCLighting (Variables|Default|GetIsActive)))))

(fn SwitchErebusLighting
  (params (Visibility :type Boolean))
  (Rendering|SetVisibility (Variables|Default|GetKeyLight_StandaloneForSwitch) Visibility false)
  (Rendering|Components|Light|SetCastShadows (Variables|Default|GetBounce1) false))

(fn SwitchMobileLighting
  (params (NewParam :type Boolean))
  (Rendering|SetVisibility (Variables|Default|GetRimTopRight1) false false)
  (Rendering|SetVisibility (Variables|Default|GetDirectionalLightMobile) NewParam false))

(fn SwitchPCLighting
  (params (Visibility :type Boolean))
  (Rendering|SetVisibility (Variables|Default|GetRimLeft1) Visibility false))

(fn ConstructionScript)
