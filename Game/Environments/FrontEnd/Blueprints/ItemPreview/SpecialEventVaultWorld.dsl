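; Front-end item preview world for the special event vault, animating the backdrop and floor behind previewed items.
; Background textures and colors alternate between A/B material slots, chosen by an active-slot flag that flips after each transition.
; Forward and backward timelines drive crossfade offsets on both background and floor dynamic materials.
; Other events toggle Digital Rain, streak effects and floor visibility, and shift the background glow for item details.
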(event EventOnUpdateMaterialIndex
  (params (MaterialIndex :type Integer))
  (bind _returnvalue (== MaterialIndex 0))
  (bind _returnvalue_2 (Math|Float|SelectFloat 0.0 1.0 _returnvalue))
  ;[0 0 0] On material Index change - Change from present to Digital Rain
  (Rendering|Material|SetScalarParameterValue (Variables|FortItemPreviewWorld|GetBackgroundMaterialDynamic) "DigitalRain" _returnvalue_2 :!class /Script/Engine.MaterialInstanceDynamic)
  (Rendering|Material|SetScalarParameterValue (Variables|Default|GetFloorMID) "DigitalRain" _returnvalue_2 :!class /Script/Engine.MaterialInstanceDynamic))

(event EventBeginPlay
  ;[0 0 0] BeginPlay
  (|CallParentFunction)
  (bind _returnvalue (Rendering|Material|CreateDynamicMaterialInstance (Variables|Default|GetFloorMI) "None" "None" :!class /Script/Engine.KismetMaterialLibrary))
  (Rendering|Material|SetMaterial (Variables|Default|GetFloor) 0 _returnvalue)
  (Variables|Default|SetFloorMID _returnvalue))

(event EventOnSetupBackground
  (params (LoadedBackgroundTexture :type Texture%202D%20Object%20Reference) (BackgroundInfo :type Vault%20World%20Background%20Data%20Structure%20%28by%20ref%29))
  ;[0 0 0] OnSetupBackground
  (CallFunction|SetupBackgroundMaterial LoadedBackgroundTexture BackgroundInfo (Variables|FortItemPreviewWorld|GetBackgroundMaterialDynamic))
  (CallFunction|SetupBackgroundMaterial LoadedBackgroundTexture BackgroundInfo (Variables|Default|GetFloorMID)))

(event EventOnTransitionBackground
  (params (TransitionDirection :type EVaultWorldTransitionDirection%20Enum :default "Forward") (IntensityTransition :type EBackgroundIntensityLevel%20Enum))
  ;[0 0 0] OnTransitionBackground
  (switch Utilities|FlowControl|Switch|SwitchonEVaultWorldTransitionDirection TransitionDirection
    (:Forward
      (bind (_direction _forward _pre_transitionoffset _fx_transition_fade) (Timeline|TransitionForward PlayFromStart)
        (:Update
          (Utilities|FlowControl|Sequence
            (:then_0
              (CallFunction|TransitionBackgroundForward _forward _pre_transitionoffset _fx_transition_fade (Variables|FortItemPreviewWorld|GetBackgroundMaterialDynamic)))
            (:then_1
              (CallFunction|TransitionBackgroundForward _forward _pre_transitionoffset _fx_transition_fade (Variables|Default|GetFloorMID)))))
        (:Finished
          (Variables|Default|SetbIsPrimaryBackgroundActive (not (Variables|Default|GetbIsPrimaryBackgroundActive))))))
    (:Backward
      (bind (_direction_2 _backward _pre_transition_offset _fx_transition_fade_2) (Timeline|TransitionBackward PlayFromStart)
        (:Update
          (Utilities|FlowControl|Sequence
            (:then_0
              (CallFunction|TransitionBackgroundBackward _backward _pre_transition_offset _fx_transition_fade_2 (Variables|FortItemPreviewWorld|GetBackgroundMaterialDynamic)))
            (:then_1
              (CallFunction|TransitionBackgroundBackward _backward _pre_transition_offset _fx_transition_fade_2 (Variables|Default|GetFloorMID)))))
        (:Finished
          (Variables|Default|SetbIsPrimaryBackgroundActive (not (Variables|Default|GetbIsPrimaryBackgroundActive))))))
    (:None
      ;[0 0 0] OnFirstTimeBackgroundSetup
      (Utilities|FlowControl|Sequence
        (:then_0
          (CallFunction|TransitionBackgroundForward 1.0 0.0 1.0 (Variables|FortItemPreviewWorld|GetBackgroundMaterialDynamic)))
        (:then_1
          (CallFunction|TransitionBackgroundForward 1.0 0.0 1.0 (Variables|Default|GetFloorMID)))
        (:then_2
          (Variables|Default|SetbIsPrimaryBackgroundActive (not (Variables|Default|GetbIsPrimaryBackgroundActive))))))))

(event EventOnUpdateDisplay
  (params (bShowFloor :type Boolean) (bShowEffects :type Boolean) (bIsOnPawn :type Boolean) (ItemType :type EFortItemType%20Enum :default "WorldItem"))
  ;[0 0 0] OnShowBackgroundEffects
  (Utilities|FlowControl|Sequence
    (:then_0
      (if bShowEffects
        (bind (_direction _setstreaks) (Timeline|Background-Effects Reverse)
          (:Update
            (Rendering|Material|SetScalarParameterValue (Variables|FortItemPreviewWorld|GetBackgroundMaterialDynamic) "bIsUseStreak" _setstreaks :!class /Script/Engine.MaterialInstanceDynamic)
            (Rendering|Material|SetScalarParameterValue (Variables|Default|GetFloorMID) "bIsUseStreak" _setstreaks :!class /Script/Engine.MaterialInstanceDynamic)))
        (else
          (Timeline|Background-Effects Play))))
    (:then_1
      (if bShowFloor
        (bind (_direction_2 _floormask) (Timeline|Floor-Visibility Play)
          (:Update
            (Rendering|Material|SetScalarParameterValue (Variables|Default|GetFloorMID) "Global-FloorMask-Density" _floormask :!class /Script/Engine.MaterialInstanceDynamic))
          (:Finished
            (Rendering|SetVisibility (Variables|Default|GetFloor) bShowFloor true)))
        (else
          (Timeline|Floor-Visibility Reverse))))))

(event EventOnTransitionItemDetails
  (params (bShowItemDetails :type Boolean))
  ;[0 0 0] OnTransitionItemDetails
  (if bShowItemDetails
    (bind (_direction _x_offset) (Timeline|ItemDetails Play)
      (:Update
        (Rendering|Material|SetScalarParameterValue (Variables|FortItemPreviewWorld|GetBackgroundMaterialDynamic) "Cnt_Glow_x-Offset" _x_offset :!class /Script/Engine.MaterialInstanceDynamic)))
    (else
      (Timeline|ItemDetails Reverse))))

(fn SetupBackgroundMaterial
  (params (TextureBackground :type Texture%202D%20Object%20Reference%20%28by%20ref%29) (BackgroundInfo :type Vault%20World%20Background%20Data%20Structure%20%28by%20ref%29) (MID :type Material%20Instance%20Dynamic%20Object%20Reference))
  (bind (_backgroundsubstance _backgroundoffset _primarycolor _secondarycolor) (Utilities|Struct|BreakVaultWorldBackgroundData BackgroundInfo))
  (bind _bisprimarybackgroundactive (Variables|Default|GetbIsPrimaryBackgroundActive))
  (Utilities|IsValid TextureBackground
    (:Is%20Valid
      (Rendering|Material|SetTextureParameterValue MID (select _bisprimarybackgroundactive "bBackingTexture" "aBackingTexture") TextureBackground))
    (:Is%20Not%20Valid
      _))
  (Rendering|Material|SetVectorParameterValue MID (select _bisprimarybackgroundactive "bRGB1" "aRGB1") _primarycolor :!class /Script/Engine.MaterialInstanceDynamic)
  (Rendering|Material|SetVectorParameterValue MID (select _bisprimarybackgroundactive "bRGB2" "aRGB2") _secondarycolor :!class /Script/Engine.MaterialInstanceDynamic))

(fn TransitionBackgroundBackward
  (params (Backward :type Float%20%28double-precision%29) (PreTransitionOffset :type Float%20%28double-precision%29) (FXTransitionFade :type Float%20%28double-precision%29) (MID :type Material%20Instance%20Dynamic%20Object%20Reference))
  (bind _pretransitionoffset (Variables|GetPreTransitionOffset))
  (bind _fxtransitionfade (Variables|GetFXTransitionFade))
  (bind _backward (Variables|GetBackward))
  (if (Variables|Default|GetbIsPrimaryBackgroundActive)
    ;[0 0 0 1 12] Forward A to B
    (Rendering|Material|SetScalarParameterValue (Variables|GetMID) "Transition-Offset" (Math|Float|Lerp 0.0 1.0 _backward) :!class /Script/Engine.MaterialInstanceDynamic)
    ;[0 0 0 1 12] Offset Textures
    (Rendering|Material|SetScalarParameterValue (Variables|GetMID) "Transition-Offset-aBacking" (Math|Float|Lerp 0.0 -0.125 _backward) :!class /Script/Engine.MaterialInstanceDynamic)
    (Rendering|Material|SetScalarParameterValue (Variables|GetMID) "Transition-Offset-bBacking" (Math|Float|Lerp 0.125 0.0 _backward) :!class /Script/Engine.MaterialInstanceDynamic)
    ;[0 0 0 1 12] Offset Pre-Mask
    (Rendering|Material|SetScalarParameterValue (Variables|GetMID) "Pre-Transition-Offset" (Math|Float|Lerp 0.0 -0.125 _pretransitionoffset) :!class /Script/Engine.MaterialInstanceDynamic)
    ;[0 0 0 1 12] Fade FX
    (Rendering|Material|SetScalarParameterValue (Variables|GetMID) "fx-Transition-Fade" _fxtransitionfade :!class /Script/Engine.MaterialInstanceDynamic)
    (else
      ;[0 0 0 1 12] Forward B to A
      (Rendering|Material|SetScalarParameterValue (Variables|GetMID) "Transition-Offset" (Math|Float|Lerp 1.0 0.0 _backward) :!class /Script/Engine.MaterialInstanceDynamic)
      (Rendering|Material|SetScalarParameterValue (Variables|GetMID) "Transition-Offset-aBacking" (Math|Float|Lerp 0.125 0.0 _backward) :!class /Script/Engine.MaterialInstanceDynamic)
      (Rendering|Material|SetScalarParameterValue (Variables|GetMID) "Transition-Offset-bBacking" (Math|Float|Lerp 0.0 -0.125 _backward) :!class /Script/Engine.MaterialInstanceDynamic)
      (Rendering|Material|SetScalarParameterValue (Variables|GetMID) "Pre-Transition-Offset" (Math|Float|Lerp 0.0 0.125 _pretransitionoffset) :!class /Script/Engine.MaterialInstanceDynamic)
      (Rendering|Material|SetScalarParameterValue (Variables|GetMID) "fx-Transition-Fade" _fxtransitionfade :!class /Script/Engine.MaterialInstanceDynamic))))

(fn TransitionBackgroundForward
  (params (Forward :type Float%20%28double-precision%29) (PreTransitionOffset :type Float%20%28double-precision%29) (FXTransitionFade :type Float%20%28double-precision%29) (MID :type Material%20Instance%20Dynamic%20Object%20Reference))
  (bind _pretransitionoffset (Variables|GetPreTransitionOffset))
  (bind _fxtransitionfade (Variables|GetFXTransitionFade))
  (bind _forward (Variables|GetForward))
  (if (Variables|Default|GetbIsPrimaryBackgroundActive)
    ;[0 0 0 1 12] Forward A to B
    (Rendering|Material|SetScalarParameterValue (Variables|GetMID) "Transition-Offset" (Math|Float|Lerp 0.0 1.0 _forward) :!class /Script/Engine.MaterialInstanceDynamic)
    ;[0 0 0 1 12] Offset Textures
    (Rendering|Material|SetScalarParameterValue (Variables|GetMID) "Transition-Offset-aBacking" (Math|Float|Lerp 0.0 0.125 _forward) :!class /Script/Engine.MaterialInstanceDynamic)
    (Rendering|Material|SetScalarParameterValue (Variables|GetMID) "Transition-Offset-bBacking" (Math|Float|Lerp -0.125 0.0 _forward) :!class /Script/Engine.MaterialInstanceDynamic)
    ;[0 0 0 1 12] Offset Pre-Mask
    (Rendering|Material|SetScalarParameterValue (Variables|GetMID) "Pre-Transition-Offset" (Math|Float|Lerp 0.0 0.125 _pretransitionoffset) :!class /Script/Engine.MaterialInstanceDynamic)
    ;[0 0 0 1 12] Fade FX
    (Rendering|Material|SetScalarParameterValue (Variables|GetMID) "fx-Transition-Fade" _fxtransitionfade :!class /Script/Engine.MaterialInstanceDynamic)
    (else
      ;[0 0 0 1 12] Forward B to A
      (Rendering|Material|SetScalarParameterValue (Variables|GetMID) "Transition-Offset" (Math|Float|Lerp 1.0 0.0 _forward) :!class /Script/Engine.MaterialInstanceDynamic)
      (Rendering|Material|SetScalarParameterValue (Variables|GetMID) "Transition-Offset-aBacking" (Math|Float|Lerp -0.125 0.0 _forward) :!class /Script/Engine.MaterialInstanceDynamic)
      (Rendering|Material|SetScalarParameterValue (Variables|GetMID) "Transition-Offset-bBacking" (Math|Float|Lerp 0.0 0.125 _forward) :!class /Script/Engine.MaterialInstanceDynamic)
      (Rendering|Material|SetScalarParameterValue (Variables|GetMID) "Pre-Transition-Offset" (Math|Float|Lerp 0.0 -0.125 _pretransitionoffset) :!class /Script/Engine.MaterialInstanceDynamic)
      (Rendering|Material|SetScalarParameterValue (Variables|GetMID) "fx-Transition-Fade" _fxtransitionfade :!class /Script/Engine.MaterialInstanceDynamic))))

(fn ConstructionScript
  (Rendering|Material|SetMaterial (Variables|Default|GetFloor) 0 (Variables|Default|GetFloorMI))
  (bind _asfort_static_mesh_actor (Utilities|Casting|CastToFortStaticMeshActor (Variables|/Script/Engine.ChildActorComponent|GetChildActor (Variables|FortItemPreviewWorld|GetSceneBackgroundComponent))))
  (Rendering|Material|SetMaterial (Variables|/Script/Engine.StaticMeshActor|GetStaticMeshComponent _asfort_static_mesh_actor) 0 (Variables|FortItemPreviewWorld|GetBackgroundMaterial))
  (|CallParentFunction))
