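; VaultWorld is a front-end item preview world Blueprint that manages dynamic background materials and relays background text toggles.
; On BeginPlay and whenever a new scene background child actor arrives, it creates dynamic instances of the background and floor materials.
; The background instance goes onto slot 0 of that actor's mesh; both instances are handed to setters, with the floor one not applied here.
; Text toggles forward to the VaultRotator child actor of the scene changer component.
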(event EventBeginPlay
  (bind _asfort_static_mesh_actor (Utilities|Casting|CastToFortStaticMeshActor (Variables|/Script/Engine.ChildActorComponent|GetChildActor (Variables|FortItemPreviewWorld|GetSceneBackgroundComponent))))
  (CallFunction|AssignBackgroundDynamicMaterials _asfort_static_mesh_actor))

(event EventOnNewSceneBackgroundChildActor
  (params (NewActor :type Fort%20Static%20Mesh%20Actor%20Object%20Reference))
  (CallFunction|AssignBackgroundDynamicMaterials NewActor))

(event EventToggleBackgroundText
  (params (DisplayText :type Boolean))
  (bind _vaultrotator (CallFunction|GetVaultRotator))
  (Class|VaultRotator|ToggleBackgroundText _vaultrotator DisplayText))

(fn AssignBackgroundDynamicMaterials
  (params (TargetBackground :type Static%20Mesh%20Actor%20Object%20Reference))
  (bind _returnvalue (Rendering|Material|CreateDynamicMaterialInstance (Variables|FortItemPreviewWorld|GetBackgroundMaterial) "None" "None" :!class /Script/Engine.KismetMaterialLibrary))
  (Rendering|Material|SetMaterial (Variables|/Script/Engine.StaticMeshActor|GetStaticMeshComponent TargetBackground) 0 _returnvalue)
  (CallFunction|SetBackgroundMaterialDynamic _returnvalue)
  (bind _returnvalue_2 (Rendering|Material|CreateDynamicMaterialInstance (Variables|FortItemPreviewWorld|GetFloorMaterial) "None" "None" :!class /Script/Engine.KismetMaterialLibrary))
  (CallFunction|SetFloorMaterialDynamic _returnvalue_2))

(fn GetVaultRotator
  (bind _asvault_rotator (Utilities|Casting|CastToVaultRotator (Variables|/Script/Engine.ChildActorComponent|GetChildActor (Variables|FortItemPreviewWorld|GetSceneChangerComponent))))
  (return _asvault_rotator))

(fn ConstructionScript)
