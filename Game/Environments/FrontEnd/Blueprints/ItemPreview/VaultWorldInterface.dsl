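; VaultWorldInterface is a Blueprint interface for the front-end item preview environment, going by its location and name.
; It declares one function, ToggleBackgroundText, which takes a Boolean DisplayText to show or hide background text.
; The file holds only the signature, so implementing objects supply the behavior.
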(fn ToggleBackgroundText
  (params (DisplayText :type Boolean)))
