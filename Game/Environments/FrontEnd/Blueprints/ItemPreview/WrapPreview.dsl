(event EventBeginPlay)

(event Collision|EventActorBeginOverlap
  (params (OtherActor :type Actor%20Object%20Reference)))

(event EventTick
  (params (DeltaSeconds :type Float%20%28single-precision%29)))

(event EventOnPreviewVisualsSpawned
  (Rendering|SetActorHiddenInGame true)
  (Niagara|SpawnSystematLocation "/Game/Effects/Fort_Effects/UI/Transition/P_FrontEndCharacterTransition_Converted.P_FrontEndCharacterTransition_Converted" (+ (Transformation|GetActorLocation) "0.000000,0.000000,50.000000") "0.000000,90.000000,0.000000" "0.500000,0.500000,0.500000")
  (Audio|PlaySoundatLocation "/Game/Sounds/Fort_UI_Sounds/S13_Frontend/BattlePass_ResIn_Cue.BattlePass_ResIn_Cue" (Transformation|GetActorLocation))
  (Utilities|FlowControl|Delay 0.1)
  (Rendering|SetActorHiddenInGame false)
  (CallFunction|UpdateLightingScale)
  (CallFunction|UpdateFloorVisibility)
  (if (not (Variables|Default|GetAlwaysOn))
    (Utilities|FlowControl|Sequence
      (:then_0
        (Rendering|Components|Light|SetForceCachedShadowsforMovablePrimitives (Variables|Default|GetRimBottomLeft2) true))
      (:then_1
        ;[0 0 0 1 12] Force Disable on start
        (CallFunction|SwitchMobileLighting false)
        (CallFunction|SwitchPCLighting false)
        (CallFunction|LightControl true))
      (:then_2
        ;[0 0 0 1 12] Sitch lighting if settings are switched while screen is already active
        (bind _returnvalue (Settings|GetGameUserSettings))
        (EventDispatchers|BindEventtoOnGameUserSettingsUINeedsUpdate _returnvalue (delegate Custom|UpdateSettings))))))

(event Custom|UpdateSettings
  (bind _returnvalue (TouchPlatform|IsNintendoSwitch))
  (bind _returnvalue_2 (TouchPlatform|IsMobileGame))
  (bind _returnvalue_3 (Settings|GetFortGameUserSettings))
  (bind _returnvalue_4 (Class|FortGameUserSettings|IsinPerformanceMode _returnvalue_3))
  (bind _returnvalue_5 (Math|Boolean|ORBoolean _returnvalue_2 _returnvalue_4 _returnvalue))
  ;[0 0 0 1 12] Sitch lighting if settings are switched while screen is already active
  (if (Variables|Default|GetIsActive)
    (if (or (== (Settings|GetShadowQuality (stale EventOnPreviewVisualsSpawned _returnvalue)) 0) _returnvalue_5)
      (if _returnvalue_5
        ;;[0.125 0.125 0.125 1 12] Enable Mobile
        (CallFunction|SwitchPCLighting false)
        (CallFunction|SwitchMobileLighting true)
        (else
          ;;[0.125 0.125 0.125 1 12] Enable Low Detail PC
          (CallFunction|SwitchPCLighting true)))
      (else
        ;;[0.125 0.125 0.125 1 12] Enable HighEnd
        (CallFunction|SwitchMobileLighting false)
        (CallFunction|SwitchPCLighting true)))))

(event EventOnSetFloorMaterial
  (params (InMaterialInstance :type Material%20Interface%20Object%20Reference))
  ;[0 0 0] OnSetFloorMaterial
  (bind _returnvalue (Rendering|Material|CreateDynamicMaterialInstance (Variables|FortItemPreviewActor|GetSpecialEventMaterial) "None" "None" :!class /Script/Engine.KismetMaterialLibrary))
  (Rendering|Material|SetMaterial (Variables|Default|GetPlane) 0 _returnvalue)
  (Rendering|Material|CopyMaterialInstanceParameters _returnvalue InMaterialInstance false))

(event EventOnUpdateFloorMaterial
  (params (bEnableAutotestBackground :type Boolean))
  (if bEnableAutotestBackground
    (bind _returnvalue (Rendering|Material|CreateDynamicMaterialInstance (Variables|FortItemPreviewMultiAngleActor|GetAutotestFloorMaterial) "None" "None" :!class /Script/Engine.KismetMaterialLibrary))
    (Rendering|Material|SetMaterial (Variables|Default|GetPlane) 0 _returnvalue)
    (else
      (Rendering|Material|SetMaterial (Variables|Default|GetPlane) 0 "/Game/UI/Foundation/Environment/Shell/Materials/M_UI_FloorShadow.M_UI_FloorShadow"))))

(event EventOnItemsPendingTransitionOut
  (params (TransitionController :type State%20Transition%20Controller%20Handle%20Structure%20%28by%20ref%29)))

(event Custom|Internal_ItemsPendingTransitionOutComplete
  (params (TransitionController :type State%20Transition%20Controller%20Handle%20Structure))
  (CallFunction|OnItemsPendingTransitionOutEnsureAllResOutEffectsAreCleanedUp))

(event Custom|OnItemsPendingTransitionOut_EnsureAllResOutEffectsAreCleanedUp)

(event Custom|OnItemsPendingTransitionOut_SpawnResOutEffects)

(event Custom|FinishShowingResOutEffects
  ;[0 1 0.9312] Call this when Res-Out is done and you're ready to see the next item
  (Class|StateTransitionControlLibrary|ClearPauseRequest (Variables|Default|GetItemsPendingTransitionOutHandle) "ClearRightAway"))

(fn HandleLightingScale
  (bind _deltascale (Variables|Default|GetDeltaScale))
  (bind _deltascale_2 (Variables|Default|GetDeltaScale))
  (bind _deltascale_3 (Variables|Default|GetDeltaScale))
  (Utilities|FlowControl|Sequence
    (:then_0
      (Transformation|SetWorldScale3D (Variables|Default|GetLightingPivot) (* (Transformation|GetWorldScale (Variables|Default|GetLightingPivot)) (Variables|Default|GetDeltaScale))))
    (:then_1
      (Rendering|Lighting|SetAttenuationRadius (Variables|Default|GetRimBottomLeft2) (* (Variables|/Script/Engine.LocalLightComponent|GetAttenuationRadius (Variables|Default|GetRimBottomLeft2)) _deltascale))
      (Rendering|Lighting|SetAttenuationRadius (Variables|Default|GetRimTopRight2) (* (Variables|/Script/Engine.LocalLightComponent|GetAttenuationRadius (Variables|Default|GetRimTopRight2)) _deltascale))
      (Rendering|Lighting|SetAttenuationRadius (Variables|Default|GetRimLeft2) (* (Variables|/Script/Engine.LocalLightComponent|GetAttenuationRadius (Variables|Default|GetRimLeft2)) _deltascale))
      (Rendering|Lighting|SetAttenuationRadius (Variables|Default|GetRimLowerRight2) (* (Variables|/Script/Engine.LocalLightComponent|GetAttenuationRadius (Variables|Default|GetRimLowerRight2)) _deltascale)))
    (:then_2
      (Rendering|Lighting|SetAttenuationRadius (Variables|Default|GetBounceRear2) (* (Variables|/Script/Engine.LocalLightComponent|GetAttenuationRadius (Variables|Default|GetBounceRear2)) _deltascale_2))
      (Rendering|Lighting|SetAttenuationRadius (Variables|Default|GetBounce2) (* (Variables|/Script/Engine.LocalLightComponent|GetAttenuationRadius (Variables|Default|GetBounce2)) _deltascale_2)))
    (:then_3
      (Rendering|Lighting|SetAttenuationRadius (Variables|Default|GetKeyLight3) (* (Variables|/Script/Engine.LocalLightComponent|GetAttenuationRadius (Variables|Default|GetKeyLight3)) _deltascale_3))
      (Rendering|Lighting|SetAttenuationRadius (Variables|Default|GetKeyLight8) (* (Variables|/Script/Engine.LocalLightComponent|GetAttenuationRadius (Variables|Default|GetKeyLight8)) _deltascale_3))
      (Rendering|Lighting|SetAttenuationRadius (Variables|Default|GetKeyLight7) (* (Variables|/Script/Engine.LocalLightComponent|GetAttenuationRadius (Variables|Default|GetKeyLight7)) _deltascale_3))
      (Rendering|Lighting|SetAttenuationRadius (Variables|Default|GetKeyLigh4) (* (Variables|/Script/Engine.LocalLightComponent|GetAttenuationRadius (Variables|Default|GetKeyLigh4)) _deltascale_3)))))

(fn LightControl
  (params (Active :type Boolean))
  (bind _returnvalue (TouchPlatform|IsNintendoSwitch))
  (bind _returnvalue_2 (Settings|GetFortGameUserSettings))
  (bind _returnvalue_3 (Class|FortGameUserSettings|IsinPerformanceMode _returnvalue_2))
  (bind _returnvalue_4 (TouchPlatform|IsMobileGame))
  (bind _returnvalue_5 (TouchPlatform|IsNintendoSwitch))
  (bind _returnvalue_6 (Math|Boolean|ORBoolean _returnvalue_4 _returnvalue_3 _returnvalue_5))
  (bind _isactive (Variables|Default|GetIsActive))
  (if (not (Variables|Default|GetAlwaysOn))
    (Variables|Default|SetIsActive Active)
    (bind _returnvalue_7 (Settings|GetGameUserSettings))
    (if (Math|Boolean|ORBoolean (== (Settings|GetShadowQuality _returnvalue_7) 0) _returnvalue_6 _returnvalue)
      (if _returnvalue_6
        (CallFunction|SwitchMobileLighting _isactive)
        (elif _returnvalue
          (CallFunction|SwitchErebusLighting _isactive))))
    (CallFunction|SwitchPCLighting (Variables|Default|GetIsActive))))

(fn SetFloorEnabled
  (params (Floor%20Enabled :type Boolean))
  (Variables|Default|SetFloorEverAllowed Floor%20Enabled)
  (CallFunction|UpdateFloorVisibility))

(fn SwitchErebusLighting
  (params (Visibility :type Boolean))
  (Rendering|SetVisibility (Variables|Default|GetKeyLight_StandaloneForSwitch) Visibility false)
  (Rendering|Components|Light|SetCastShadows (Variables|Default|GetBounce2) false))

(fn SwitchMobileLighting
  (params (NewParam :type Boolean))
  (Rendering|SetVisibility (Variables|Default|GetRimLowerRight2) false false)
  (Rendering|SetVisibility (Variables|Default|GetDirectionalLightMobile) NewParam false))

(fn SwitchPCLighting
  (params (Visibility :type Boolean))
  (Rendering|SetVisibility (Variables|Default|GetRimLowerRight2) Visibility false))

(fn UpdateFloorVisibility
  (bind _plane (Variables|Default|GetPlane))
  (if (or (CallFunction|IsShowingVehicle) (not (Variables|Default|GetFloorEverAllowed)))
    (Development|SetHiddeninGame _plane true false)
    (else
      (Development|SetHiddeninGame _plane false false))))

(fn UpdateLightingScale
  (if (CallFunction|IsShowingVehicle)
    (Variables|Default|SetDeltaScale (Math|Float|SelectFloat 1.0 (LocalVariables|GetUpscaleForVehicle) (Variables|Default|GetbLightsScaledForVehicle)))
    (Variables|Default|SetbLightsScaledForVehicle true)
    (CallFunction|HandleLightingScale)
    (else
      (Variables|Default|SetDeltaScale (Math|Float|SelectFloat (/ 1.0 (LocalVariables|GetUpscaleForVehicle)) 1.0 (Variables|Default|GetbLightsScaledForVehicle)))
      (Variables|Default|SetbLightsScaledForVehicle false)
      (CallFunction|HandleLightingScale))))

(fn ConstructionScript)
